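# Sound cue for equipping the repair tool, playing a single wave through a Modulator left at its default settings for slight random pitch and volume variation.
# It reuses pickup audio setup: the loot-collect sound class and the PickupMaximum_Oldest concurrency set.
# Its own attenuation override shapes spatial falloff in place of the referenced pickup attenuation asset.
# A distance-scaled linear send feeds the distant reverb submix.
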SoundCue /ResolveGameplay/Items/RepairTool/Audio/RepairTool_Equip_Cue
root=Modulator_1
nodes [
 {id=Modulator_1, children [WavePlayer_6]}
 {id=WavePlayer_6, properties {soundWaveAssetPtr=/ResolveGameplay/Items/RepairTool/Audio/wav/Repair_Tool_Equip}}
]
properties {
 attenuationSettings=&Fort_Weapon_Pickup_Att
 bOverrideAttenuation=true
 maxDistance=1512
 soundClassObject=&Fort_Loot_Collect
 volumeMultiplier=0.259
 attenuationOverrides {falloffDistance=1000, stereoSpread=150, attenuationShapeExtents {x=512}}
 concurrencyOverrides {maxCount=1, resolutionRule=PreventNew}
 concurrencySet [&PickupMaximum_Oldest]
 soundSubmixSends [{
   maxSendDistance=7500
   maxSendLevel=0.25
   minSendDistance=0
   minSendLevel=0
   sendLevel=0
   sendLevelControlMethod=Linear
   soundSubmix=&Fortnite_Reverb_Distant
   customSendLevelCurve {externalCurve=/Game/Sounds/SubMixes/Reverb/Fortnite_Reverb_Distant_Send}
  }]
}
