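# Sound cue that supplies the foley for the repair tool's aim-in.
# A single wave player, using a wave named as a Convoy Tarantula run footstep, feeds a modulator left at default settings.
# The cue plays in the player foley sound class with close-prop attenuation.
# Its own concurrency override prevents new instances once the limit is reached rather than stopping playing ones.
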SoundCue /ResolveGameplay/Items/RepairTool/Audio/RepairTool_Foley_Aim_In_Cue
root=Modulator_2
nodes [
 {id=Modulator_2, children [WavePlayer_1]}
 {id=WavePlayer_1, properties {soundWaveAssetPtr=/ResolveGameplay/Items/RepairTool/Audio/wav/FS_ConvoyTarantula_Default_Run_13}}
]
properties {
 attenuationSettings=/Game/Sounds/Attenuation/Fort_Prop_Close_Att
 bOverrideConcurrency=true
 maxDistance=1800
 soundClassObject=&Fort_Player_Foley
 volumeMultiplier=1
 concurrencyOverrides {resolutionRule=PreventNew}
}
