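# Sound cue for the repair tool's default impact one-shot, playing a short electrical spark.
# A Random node picks one of four equally weighted big-spark waves, an Enveloper fades it out by 0.4 seconds, and a Modulator varies each play.
# It plays quietly on the ambient tile-mix sound class with overridden attenuation.
# A single-voice concurrency override blocks new instances while one is playing.
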SoundCue /ResolveGameplay/Items/RepairTool/Audio/RepairTool_IMPT_OS_Default_Cue
root=Modulator_0
nodes [
 {
  id=Enveloper_0
  children [Random_0]
  properties {pitchCurve {editorCurveData {keys [{time=0}]}}, volumeCurve {editorCurveData {keys [<at 0 1>,<at 0.2494 1>,<at 0.4 0>]}}}
 }
 {id=Modulator_0, children [Enveloper_0]}
 {id=Random_0, children [WavePlayer_2,WavePlayer_3,WavePlayer_4,WavePlayer_5], properties {weights [1,1,1,1]}}
 {id=WavePlayer_2, properties {soundWaveAssetPtr=/Game/Sounds/Fort_Atlas_Sounds/atlas_sparkbig_mono_01}}
 {id=WavePlayer_3, properties {soundWaveAssetPtr=/Game/Sounds/Fort_Atlas_Sounds/atlas_sparkbig_mono_02}}
 {id=WavePlayer_4, properties {soundWaveAssetPtr=/Game/Sounds/Fort_Atlas_Sounds/atlas_sparkbig_mono_03}}
 {id=WavePlayer_5, properties {soundWaveAssetPtr=/Game/Sounds/Fort_Atlas_Sounds/atlas_sparkbig_mono_04}}
]
properties {
 attenuationSettings=&Fort_Traps_Att
 bOverrideAttenuation=true
 bOverrideConcurrency=true
 maxDistance=800
 soundClassObject=/Game/Sounds/FortSoundClassesAndModes/Fort_Ambient_TileMix
 volumeMultiplier=0.225
 attenuationOverrides {falloffDistance=400}
 concurrencyOverrides {maxCount=1, resolutionRule=PreventNew}
}
