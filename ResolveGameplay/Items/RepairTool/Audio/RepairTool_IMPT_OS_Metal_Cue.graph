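# Sound cue for the repair tool, apparently its one-shot impact on metal, built from electric spark samples.
# A mixer layers a random pick from five spark waves with a quieter random pick from twelve others, then a root modulator adds variation.
# Attenuation and concurrency overrides set a short falloff and cap simultaneous instances, preventing new ones once the limit is reached.
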SoundCue /ResolveGameplay/Items/RepairTool/Audio/RepairTool_IMPT_OS_Metal_Cue
root=Modulator_2
nodes [
 {id=Mixer_1, children [Random_3,Random_2], properties {inputVolume [1,0.65]}}
 {id=Modulator_2, children [Mixer_1]}
 {
  id=Random_2
  children [
   WavePlayer_13
   WavePlayer_14
   WavePlayer_15
   WavePlayer_16
   WavePlayer_17
   WavePlayer_18
   WavePlayer_19
   WavePlayer_20
   WavePlayer_21
   WavePlayer_22
   WavePlayer_23
   WavePlayer_24
  ]
  properties {weights [1,1,1,1,1,1,1,1,1,1,1,1]}
 }
 {id=Random_3, children [WavePlayer_8, WavePlayer_9, WavePlayer_10, WavePlayer_11, WavePlayer_12], properties {weights [1,1,1,1,1]}}
 {id=WavePlayer_10, properties {soundWaveAssetPtr=/ResolveGameplay/Items/RepairTool/Audio/wav/Repair_Tool_Electriic_Sparks_004}}
 {id=WavePlayer_11, properties {soundWaveAssetPtr=/ResolveGameplay/Items/RepairTool/Audio/wav/Repair_Tool_Electriic_Sparks_008}}
 {id=WavePlayer_12, properties {soundWaveAssetPtr=/ResolveGameplay/Items/RepairTool/Audio/wav/Repair_Tool_Electriic_Sparks_009}}
 {id=WavePlayer_13, properties {soundWaveAssetPtr=/ResolveGameplay/Items/RepairTool/Audio/wav/Repair_Tool_Electric_Sparks_22}}
 {id=WavePlayer_14, properties {soundWaveAssetPtr=/ResolveGameplay/Items/RepairTool/Audio/wav/Repair_Tool_Electric_Sparks_21}}
 {id=WavePlayer_15, properties {soundWaveAssetPtr=/ResolveGameplay/Items/RepairTool/Audio/wav/Repair_Tool_Electric_Sparks_20}}
 {id=WavePlayer_16, properties {soundWaveAssetPtr=/ResolveGameplay/Items/RepairTool/Audio/wav/Repair_Tool_Electric_Sparks_19}}
 {id=WavePlayer_17, properties {soundWaveAssetPtr=/ResolveGameplay/Items/RepairTool/Audio/wav/Repair_Tool_Electric_Sparks_18}}
 {id=WavePlayer_18, properties {soundWaveAssetPtr=/ResolveGameplay/Items/RepairTool/Audio/wav/Repair_Tool_Electric_Sparks_17}}
 {id=WavePlayer_19, properties {soundWaveAssetPtr=/ResolveGameplay/Items/RepairTool/Audio/wav/Repair_Tool_Electric_Sparks_16}}
 {id=WavePlayer_20, properties {soundWaveAssetPtr=/ResolveGameplay/Items/RepairTool/Audio/wav/Repair_Tool_Electric_Sparks_15}}
 {id=WavePlayer_21, properties {soundWaveAssetPtr=/ResolveGameplay/Items/RepairTool/Audio/wav/Repair_Tool_Electric_Sparks_14}}
 {id=WavePlayer_22, properties {soundWaveAssetPtr=/ResolveGameplay/Items/RepairTool/Audio/wav/Repair_Tool_Electric_Sparks_13}}
 {id=WavePlayer_23, properties {soundWaveAssetPtr=/ResolveGameplay/Items/RepairTool/Audio/wav/Repair_Tool_Electric_Sparks_12}}
 {id=WavePlayer_24, properties {soundWaveAssetPtr=/ResolveGameplay/Items/RepairTool/Audio/wav/Repair_Tool_Electric_Sparks_11}}
 {id=WavePlayer_8, properties {soundWaveAssetPtr=/ResolveGameplay/Items/RepairTool/Audio/wav/Repair_Tool_Electriic_Sparks_002}}
 {id=WavePlayer_9, properties {soundWaveAssetPtr=/ResolveGameplay/Items/RepairTool/Audio/wav/Repair_Tool_Electriic_Sparks_003}}
]
properties {
 attenuationSettings=&Fort_Traps_Att
 bOverrideAttenuation=true
 bOverrideConcurrency=true
 maxDistance=800
 soundClassObject=/Game/Sounds/FortSoundClassesAndModes/Fort_Ambient_TileMix
 volumeMultiplier=1
 attenuationOverrides {falloffDistance=400}
 concurrencyOverrides {maxCount=3, resolutionRule=PreventNew}
}
