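# Sound cue for the repair tool's one-shot electric spark impact.
# A quality switch gives the first level a modulated mix of two random spark pools, the second pool quieter, for pitch and volume variation.
# The other quality levels share a lighter mix drawn from smaller random subsets of the same waves.
# It overrides attenuation and uses concurrency limits that block new instances once a few are playing.
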SoundCue /ResolveGameplay/Items/RepairTool/Audio/RepairTool_IMPT_OS_Sparks_Cue
root=QualityLevel_0
nodes [
 {id=Mixer_0, children [Random_3,Random_0], properties {inputVolume [1,0.65]}}
 {id=Mixer_2, children [Random_5,Random_6], properties {inputVolume [1,1]}}
 {id=Modulator_0, children [Mixer_0]}
 {id=QualityLevel_0, children [Modulator_0,Mixer_2,Mixer_2,Mixer_2]}
 {
  id=Random_0
  children [WavePlayer_0, WavePlayer_1, WavePlayer_2, WavePlayer_12, WavePlayer_13, WavePlayer_14]
  properties {weights [1,1,1,1,1,1]}
 }
 {id=Random_3, children [WavePlayer_4, WavePlayer_5, WavePlayer_6, WavePlayer_10, WavePlayer_11], properties {weights [1,1,1,1,1]}}
 {id=Random_5, children [WavePlayer_3,WavePlayer_7], properties {weights [1,1]}}
 {id=Random_6, children [WavePlayer_22, WavePlayer_23, WavePlayer_24], properties {weights [1,1,1]}}
 {id=WavePlayer_0, properties {soundWaveAssetPtr=/ResolveGameplay/Items/RepairTool/Audio/wav/Repair_Tool_Electric_Sparks_22}}
 {id=WavePlayer_1, properties {soundWaveAssetPtr=/ResolveGameplay/Items/RepairTool/Audio/wav/Repair_Tool_Electric_Sparks_21}}
 {id=WavePlayer_10, properties {soundWaveAssetPtr=/ResolveGameplay/Items/RepairTool/Audio/wav/Repair_Tool_Electriic_Sparks_008}}
 {id=WavePlayer_11, properties {soundWaveAssetPtr=/ResolveGameplay/Items/RepairTool/Audio/wav/Repair_Tool_Electriic_Sparks_009}}
 {id=WavePlayer_12, properties {soundWaveAssetPtr=/ResolveGameplay/Items/RepairTool/Audio/wav/Repair_Tool_Electric_Sparks_19}}
 {id=WavePlayer_13, properties {soundWaveAssetPtr=/ResolveGameplay/Items/RepairTool/Audio/wav/Repair_Tool_Electric_Sparks_18}}
 {id=WavePlayer_14, properties {soundWaveAssetPtr=/ResolveGameplay/Items/RepairTool/Audio/wav/Repair_Tool_Electric_Sparks_17}}
 {id=WavePlayer_2, properties {soundWaveAssetPtr=/ResolveGameplay/Items/RepairTool/Audio/wav/Repair_Tool_Electric_Sparks_20}}
 {id=WavePlayer_22, properties {soundWaveAssetPtr=&Repair_Tool_Electric_Sparks_22}}
 {id=WavePlayer_23, properties {soundWaveAssetPtr=&Repair_Tool_Electric_Sparks_21}}
 {id=WavePlayer_24, properties {soundWaveAssetPtr=&Repair_Tool_Electric_Sparks_20}}
 {id=WavePlayer_3, properties {soundWaveAssetPtr=/ResolveGameplay/Items/RepairTool/Audio/wav/Repair_Tool_Electriic_Sparks_002}}
 {id=WavePlayer_4, properties {soundWaveAssetPtr=&Repair_Tool_Electriic_Sparks_002}}
 {id=WavePlayer_5, properties {soundWaveAssetPtr=/ResolveGameplay/Items/RepairTool/Audio/wav/Repair_Tool_Electriic_Sparks_003}}
 {id=WavePlayer_6, properties {soundWaveAssetPtr=/ResolveGameplay/Items/RepairTool/Audio/wav/Repair_Tool_Electriic_Sparks_004}}
 {id=WavePlayer_7, properties {soundWaveAssetPtr=&Repair_Tool_Electriic_Sparks_003}}
]
properties {
 attenuationSettings=&Fort_Traps_Att
 bOverrideAttenuation=true
 bOverrideConcurrency=true
 maxDistance=800
 soundClassObject=/Game/Sounds/FortSoundClassesAndModes/Fort_Ambient_TileMix
 volumeMultiplier=0.5
 attenuationOverrides {falloffDistance=400}
 concurrencyOverrides {maxCount=3, resolutionRule=PreventNew}
}
