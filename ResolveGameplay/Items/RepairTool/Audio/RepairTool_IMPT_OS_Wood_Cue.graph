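# Sound cue for the repair tool's wood impact, layering an electric spark over a wood hit.
# A mixer blends two modulated branches: a random pick of two ceiling-trap spark waves and a random pick of seven wood impacts with a fixed lowered pitch.
# The whole cue plays at raised pitch and reduced volume.
# It overrides attenuation for a short falloff and caps concurrency, blocking new instances at the limit.
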SoundCue /ResolveGameplay/Items/RepairTool/Audio/RepairTool_IMPT_OS_Wood_Cue
root=Mixer_0
nodes [
 {id=Mixer_0, children [Modulator_0,Modulator_1], properties {inputVolume [1,1]}}
 {id=Modulator_0, children [Random_32]}
 {id=Modulator_1, children [Random_0], properties {pitchMax=0.7, pitchMin=0.7}}
 {
  id=Random_0
  children [WavePlayer_2,WavePlayer_3,WavePlayer_4,WavePlayer_5,WavePlayer_6,WavePlayer_7,WavePlayer_8]
  properties {weights [1,1,1,1,1,1,1]}
 }
 {id=Random_32, children [WavePlayer_0,WavePlayer_1], properties {weights [1,1]}}
 {id=WavePlayer_0, properties {soundWaveAssetPtr=/ResolveGameplay/Items/RepairTool/Audio/wav/CeilingTrap_Electric_Spark_002}}
 {id=WavePlayer_1, properties {soundWaveAssetPtr=/ResolveGameplay/Items/RepairTool/Audio/wav/CeilingTrap_Electric_Spark_007}}
 {id=WavePlayer_2, properties {soundWaveAssetPtr=/ResolveGameplay/Items/RepairTool/Audio/wav/wood_impact_01}}
 {id=WavePlayer_3, properties {soundWaveAssetPtr=/ResolveGameplay/Items/RepairTool/Audio/wav/wood_impact_02}}
 {id=WavePlayer_4, properties {soundWaveAssetPtr=/ResolveGameplay/Items/RepairTool/Audio/wav/wood_impact_04}}
 {id=WavePlayer_5, properties {soundWaveAssetPtr=/ResolveGameplay/Items/RepairTool/Audio/wav/wood_impact_05}}
 {id=WavePlayer_6, properties {soundWaveAssetPtr=/ResolveGameplay/Items/RepairTool/Audio/wav/wood_impact_06}}
 {id=WavePlayer_7, properties {soundWaveAssetPtr=/ResolveGameplay/Items/RepairTool/Audio/wav/wood_impact_07}}
 {id=WavePlayer_8, properties {soundWaveAssetPtr=/ResolveGameplay/Items/RepairTool/Audio/wav/wood_impact_08}}
]
properties {
 attenuationSettings=&Fort_Traps_Att
 bOverrideAttenuation=true
 bOverrideConcurrency=true
 maxDistance=800
 pitchMultiplier=3
 soundClassObject=/Game/Sounds/FortSoundClassesAndModes/Fort_Ambient_TileMix
 volumeMultiplier=0.5
 attenuationOverrides {falloffDistance=400}
 concurrencyOverrides {maxCount=3, resolutionRule=PreventNew}
}
