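# Sound cue for the repair tool landing, apparently as a dropped item hitting the ground.
# A single Repair_Tool_Land wave plays through a Modulator at default settings for slight random pitch and volume variation.
# The cue plays at reduced volume, using shared weapon-drop attenuation, the loot-collect sound class and the Weapon_Drop_StopOldest concurrency set.
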SoundCue /ResolveGameplay/Items/RepairTool/Audio/RepairTool_Landed_Cue
root=Modulator_0
nodes [
 {id=Modulator_0, children [WavePlayer_4]}
 {id=WavePlayer_4, properties {soundWaveAssetPtr=/ResolveGameplay/Items/RepairTool/Audio/wav/Repair_Tool_Land}}
]
properties {
 attenuationSettings=&Fort_Weapon_Drop_Att
 maxDistance=3000
 soundClassObject=&Fort_Loot_Collect
 volumeMultiplier=0.4
 concurrencyOverrides {resolutionRule=PreventNew}
 concurrencySet [&Weapon_Drop_StopOldest]
}
