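# Sound cue for the Repair Tool one-shot that, judging by its name, plays when a target is fully healed.
# A single Repair_Tool_Heal wave runs through a Modulator with its pitch floor raised to normal.
# The cue overrides attenuation for a short-range falloff and routes to the ambient TileMix sound class.
# A concurrency override allows one instance at a time, blocking new plays rather than cutting off the current one.
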SoundCue /ResolveGameplay/Items/RepairTool/Audio/RepairTool_OS_FullyHealed_Cue
root=Modulator_2
nodes [
 {id=Modulator_2, children [WavePlayer_12], properties {pitchMin=1}}
 {id=WavePlayer_12, properties {soundWaveAssetPtr=/ResolveGameplay/Items/RepairTool/Audio/wav/Repair_Tool_Heal}}
]
properties {
 attenuationSettings=&Fort_Traps_Att
 bOverrideAttenuation=true
 bOverrideConcurrency=true
 maxDistance=850
 soundClassObject=/Game/Sounds/FortSoundClassesAndModes/Fort_Ambient_TileMix
 volumeMultiplier=1
 attenuationOverrides {falloffDistance=400, attenuationShapeExtents {x=450}}
 concurrencyOverrides {maxCount=1, resolutionRule=PreventNew, retriggerTime=0.4}
}
