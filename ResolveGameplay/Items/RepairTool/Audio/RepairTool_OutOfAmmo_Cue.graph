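# Sound cue for the repair tool's out-of-ammo sound, built from the tool's "stop" recordings.
# A Random node picks one of three equally weighted waves and feeds a Modulator that slightly varies volume while leaving pitch fixed.
# It plays quietly in the weapon foley sound class with the repair tool attenuation, and its own concurrency override caps simultaneous instances at two by stopping the oldest.
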SoundCue /ResolveGameplay/Items/RepairTool/Audio/RepairTool_OutOfAmmo_Cue
root=Modulator_0
nodes [
 {id=Modulator_0, children [Random_1], properties {pitchMax=1, pitchMin=1, volumeMax=1.1, volumeMin=1}}
 {id=Random_1, children [WavePlayer_0,WavePlayer_4,WavePlayer_5], properties {weights [1,1,1]}}
 {id=WavePlayer_0, properties {soundWaveAssetPtr=/ResolveGameplay/Items/RepairTool/Audio/wav/Repair_Tool_Stop_01}}
 {id=WavePlayer_4, properties {soundWaveAssetPtr=/ResolveGameplay/Items/RepairTool/Audio/wav/Repair_Tool_Stop_02}}
 {id=WavePlayer_5, properties {soundWaveAssetPtr=/ResolveGameplay/Items/RepairTool/Audio/wav/Repair_Tool_Stop_03}}
]
properties {
 attenuationSettings=&/RepairTool_Att
 bOverrideConcurrency=true
 maxDistance=4000
 soundClassObject=&Fort_WeaponFoley
 volumeMultiplier=0.226
 attenuationOverrides {falloffDistance=1600}
 concurrencyOverrides {maxCount=2, resolutionRule=StopOldest}
}
