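# Sound cue for the repair tool's air idle audio, in the Fort_Consumables sound class.
# A Mixer layers a random one-shot from two loop-start and one loop-stop temp waves with a louder looping equip-idle wave.
# The loop gets slight pitch variation from a Modulator, then an Enveloper eases its volume down over several seconds to a lower level.
# Attenuation comes from RepairTool_Idle_Att, and concurrency is limited per owner, stopping the oldest voice.
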SoundCue /ResolveGameplay/Items/RepairTool/Audio/RepairTool_lp_Air_Idle_Cue
root=Mixer_0
nodes [
 {
  id=Enveloper_0
  children [Modulator_2]
  properties {
   pitchCurve {editorCurveData {keys [{time=0}]}}
   volumeCurve {editorCurveData {keys [
      {arriveTangent=-0.2288, interpMode=RCIM_Cubic, leaveTangent=-0.2288, tangentMode=RCTM_User, time=0, value=1}
      {arriveTangent=-0.1031, interpMode=RCIM_Cubic, leaveTangent=-0.1031, tangentMode=RCTM_User, time=3.3263, value=0.6099}
      <at 8 0.3>
     ]}}
  }
 }
 {id=Mixer_0, children [Random_0,Enveloper_0], properties {inputVolume [0.7,1.5]}}
 {id=Modulator_2, children [WavePlayer_1], properties {pitchMax=1, pitchMin=0.98}}
 {id=Random_0, children [WavePlayer_2,WavePlayer_3,WavePlayer_4], properties {weights [1,1,1]}}
 {id=WavePlayer_1, properties {bLooping=true, soundWaveAssetPtr=/ResolveGameplay/Items/RepairTool/Audio/wav/Repair_Tool_Equip_Idle_Loop}}
 {id=WavePlayer_2, properties {soundWaveAssetPtr=/ResolveGameplay/Items/RepairTool/Audio/wav/RepairTool_Loop_Start_Temp_01}}
 {id=WavePlayer_3, properties {soundWaveAssetPtr=/ResolveGameplay/Items/RepairTool/Audio/wav/RepairTool_Loop_Start_Temp_02}}
 {id=WavePlayer_4, properties {soundWaveAssetPtr=/ResolveGameplay/Items/RepairTool/Audio/wav/RepairTool_Loop_Stop_Temp_01}}
]
properties {
 attenuationSettings=&/RepairTool_Idle_Att
 bOverrideConcurrency=true
 maxDistance=975
 soundClassObject=&Fort_Consumables
 volumeMultiplier=2
 attenuationOverrides {bSpatialize=false, falloffDistance=650, attenuationShapeExtents {x=336.561}}
 concurrencyOverrides {
  bLimitToOwner=true
  bVolumeScaleCanRelease=true
  maxCount=2
  resolutionRule=StopOldest
  voiceStealReleaseTime=0.2
  volumeScale=0.2
  volumeScaleAttackTime=0.2
  volumeScaleReleaseTime=0.2
 }
}
