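# Looping sound cue for the repair tool's air loop while the trigger is pressed, using an activated idle loop wave.
# A looping wave player feeds a modulator that adds a very slight random downward pitch variation.
# Playback uses the repair tool attenuation and loop concurrency assets under the Fort consumables sound class.
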SoundCue /ResolveGameplay/Items/RepairTool/Audio/RepairTool_lp_Air_TriggerPressed_Cue
root=Modulator_2
nodes [
 {id=Modulator_2, children [WavePlayer_1], properties {pitchMax=1, pitchMin=0.98}}
 {id=WavePlayer_1, properties {bLooping=true, soundWaveAssetPtr=/ResolveGameplay/Items/RepairTool/Audio/wav/Repair_Tool_Activated_Idle_Loop}}
]
properties {
 attenuationSettings=&/RepairTool_Att
 maxDistance=4000
 soundClassObject=&Fort_Consumables
 volumeMultiplier=0.7
 attenuationOverrides {falloffDistance=5000, attenuationShapeExtents {x=1000}}
 concurrencyOverrides {
  bLimitToOwner=true
  bVolumeScaleCanRelease=true
  maxCount=2
  resolutionRule=StopOldest
  voiceStealReleaseTime=0.2
  volumeScale=0.2
  volumeScaleAttackTime=0.2
  volumeScaleReleaseTime=0.2
 }
 concurrencySet [&/RepairTool_Loop_Conc]
}
