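# Sound cue for the looping activated sound of the Repair Tool item, using its generic variant.
# A looping wave player feeds a modulator that slightly varies pitch, at reduced overall volume.
# It references the Repair Tool attenuation and loop concurrency assets and plays in the Fort_Consumables sound class with a custom priority.
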SoundCue /ResolveGameplay/Items/RepairTool/Audio/RepairTool_lp_Generic_Cue
root=Modulator_0
nodes [
 {id=Modulator_0, children [WavePlayer_2], properties {pitchMin=0.98}}
 {
  id=WavePlayer_2
  properties {bLooping=true, soundWaveAssetPtr=/ResolveGameplay/Items/RepairTool/Audio/wav/Repair_Tool_Activated_Generic_Loop}
 }
]
properties {
 attenuationSettings=&/RepairTool_Att
 maxDistance=4000
 priority=2
 soundClassObject=&Fort_Consumables
 volumeMultiplier=0.8
 attenuationOverrides {falloffDistance=5000, attenuationShapeExtents {x=1000}}
 concurrencyOverrides {
  bLimitToOwner=true
  bVolumeScaleCanRelease=true
  maxCount=2
  resolutionRule=StopOldest
  voiceStealReleaseTime=0.2
  volumeScale=0.2
  volumeScaleAttackTime=0.2
  volumeScaleReleaseTime=0.2
 }
 concurrencySet [&/RepairTool_Loop_Conc]
}
