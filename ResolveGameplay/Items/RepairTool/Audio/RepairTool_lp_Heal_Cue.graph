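# Looping sound cue for the Repair Tool's heal audio, routed to the Fort_Consumables sound class with RepairTool_Att attenuation.
# A looping wave player feeds a continuous modulator, so the Pitch and Volume parameters can vary the loop as it plays.
# Overridden concurrency caps it at two instances per owner, stopping the oldest with a short release and scaling down volume as instances overlap.
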SoundCue /ResolveGameplay/Items/RepairTool/Audio/RepairTool_lp_Heal_Cue
root=ModulatorContinuous_0
nodes [
 {
  id=ModulatorContinuous_0
  children [WavePlayer_1]
  properties {pitchModulationParams {minOutput=0.8, parameterName=Pitch}, volumeModulationParams {minOutput=0.5, parameterName=Volume}}
 }
 {id=WavePlayer_1, properties {bLooping=true, soundWaveAssetPtr=/ResolveGameplay/Items/RepairTool/Audio/wav/Repair_Tool_Heal_Loop}}
]
properties {
 attenuationSettings=&/RepairTool_Att
 bOverrideConcurrency=true
 maxDistance=4000
 soundClassObject=&Fort_Consumables
 volumeMultiplier=1.3
 attenuationOverrides {falloffDistance=5000, attenuationShapeExtents {x=1000}}
 concurrencyOverrides {
  bLimitToOwner=true
  bVolumeScaleCanRelease=true
  maxCount=2
  resolutionRule=StopOldest
  voiceStealReleaseTime=0.2
  volumeScale=0.2
  volumeScaleAttackTime=0.2
  volumeScaleReleaseTime=0.2
 }
}
