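# Sound cue for the repair tool's start sound, classed as a consumable and played with overridden attenuation.
# A Mixer layers a modulated random pick of three temp loop start/stop waves, boosted in the mix, over the Repair_Tool_Start wave with slight random delay and volume variation.
# The random pool includes a Loop_Stop wave alongside two Loop_Start waves.
# Overridden concurrency caps it at two voices, stopping the oldest.
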SoundCue /ResolveGameplay/Items/RepairTool/Audio/RepairTool_lp_Start_Cue
root=Mixer_0
nodes [
 {id=Delay_0, children [WavePlayer_0], properties {delayMax=0.05, delayMin=0.015}}
 {id=Mixer_0, children [Modulator_0,Modulator_2], properties {inputVolume [1.5,1]}}
 {id=Modulator_0, children [Random_0]}
 {id=Modulator_2, children [Delay_0], properties {volumeMax=1.2, volumeMin=1}}
 {id=Random_0, children [WavePlayer_1,WavePlayer_2,WavePlayer_3], properties {weights [1,1,1]}}
 {id=WavePlayer_0, properties {soundWaveAssetPtr=/ResolveGameplay/Items/RepairTool/Audio/wav/Repair_Tool_Start}}
 {id=WavePlayer_1, properties {soundWaveAssetPtr=/ResolveGameplay/Items/RepairTool/Audio/wav/RepairTool_Loop_Start_Temp_01}}
 {id=WavePlayer_2, properties {soundWaveAssetPtr=/ResolveGameplay/Items/RepairTool/Audio/wav/RepairTool_Loop_Stop_Temp_01}}
 {id=WavePlayer_3, properties {soundWaveAssetPtr=/ResolveGameplay/Items/RepairTool/Audio/wav/RepairTool_Loop_Start_Temp_02}}
]
properties {
 attenuationSettings=&/RepairTool_Att
 bOverrideAttenuation=true
 bOverrideConcurrency=true
 maxDistance=1030
 soundClassObject=&Fort_Consumables
 volumeMultiplier=1
 attenuationOverrides {falloffDistance=600, reverbWetLevelMin=0.2, stereoSpread=50, attenuationShapeExtents {x=430}}
 concurrencyOverrides {maxCount=2, resolutionRule=StopOldest, volumeScale=0.5}
 concurrencySet [/Game/Sounds/Fort_Weapons/pistol/FlareGun/FlareGun_EmberImpact_Conc]
}
