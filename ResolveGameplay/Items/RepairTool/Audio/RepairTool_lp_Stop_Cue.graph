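# RepairTool_lp_Stop_Cue is the Sound Cue for the repair tool's stop sound, likely ending its loop, in the Fort_Consumables sound class.
# A Random node picks one of three equally weighted stop waves, then a Modulator applies a fixed reduced volume and strongly raised pitch.
# Inline overrides supersede the assigned attenuation and flare gun concurrency assets, giving logarithmic falloff and a small stop-oldest voice limit with volume scaling.
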SoundCue /ResolveGameplay/Items/RepairTool/Audio/RepairTool_lp_Stop_Cue
root=Modulator_0
nodes [
 {id=Modulator_0, children [Random_1], properties {pitchMax=3, pitchMin=3, volumeMax=0.3, volumeMin=0.3}}
 {id=Random_1, children [WavePlayer_0,WavePlayer_1,WavePlayer_4], properties {weights [1,1,1]}}
 {id=WavePlayer_0, properties {soundWaveAssetPtr=/ResolveGameplay/Items/RepairTool/Audio/wav/Repair_Tool_Stop_01}}
 {id=WavePlayer_1, properties {soundWaveAssetPtr=/ResolveGameplay/Items/RepairTool/Audio/wav/Repair_Tool_Stop_02}}
 {id=WavePlayer_4, properties {soundWaveAssetPtr=/ResolveGameplay/Items/RepairTool/Audio/wav/Repair_Tool_Stop_03}}
]
properties {
 attenuationSettings=&/RepairTool_Att
 bOverrideAttenuation=true
 bOverrideConcurrency=true
 maxDistance=1030
 soundClassObject=&Fort_Consumables
 attenuationOverrides {distanceAlgorithm=Logarithmic, falloffDistance=600, reverbWetLevelMin=0.2, stereoSpread=50, attenuationShapeExtents {x=430}}
 concurrencyOverrides {maxCount=4, resolutionRule=StopOldest, volumeScale=0.5}
 concurrencySet [/Game/Sounds/Fort_Weapons/pistol/FlareGun/FlareGun_EmberImpact_Conc]
}
