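# Sound cue for the repair tool's wood variant loop audio, classed as a consumable sound with repair tool attenuation.
# A mixer layers the activated-wood loop, with slight random pitch-down variation, over the start loop, which gets a brief random delay and slight volume boost.
# Overridden concurrency limits it to two instances per owner, stopping the oldest with a short release and scaling down volume.
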SoundCue /ResolveGameplay/Items/RepairTool/Audio/RepairTool_lp_Wood_Cue
root=Mixer_0
nodes [
 {id=Delay_0, children [WavePlayer_1], properties {delayMax=0.05, delayMin=0.015}}
 {id=Mixer_0, children [Modulator_0,Modulator_1], properties {inputVolume [1,1]}}
 {id=Modulator_0, children [WavePlayer_3], properties {pitchMax=1, pitchMin=0.98}}
 {id=Modulator_1, children [Delay_0], properties {volumeMax=1.2, volumeMin=1}}
 {id=WavePlayer_1, properties {soundWaveAssetPtr=/ResolveGameplay/Items/RepairTool/Audio/wav/Repair_Tool_Start_Loop}}
 {id=WavePlayer_3, properties {soundWaveAssetPtr=/ResolveGameplay/Items/RepairTool/Audio/wav/Repair_Tool_Activated_Wood_Loop}}
]
properties {
 attenuationSettings=&/RepairTool_Att
 bOverrideConcurrency=true
 maxDistance=4000
 soundClassObject=&Fort_Consumables
 volumeMultiplier=1
 attenuationOverrides {falloffDistance=5000, attenuationShapeExtents {x=1000}}
 concurrencyOverrides {
  bLimitToOwner=true
  bVolumeScaleCanRelease=true
  maxCount=2
  resolutionRule=StopOldest
  voiceStealReleaseTime=0.2
  volumeScale=0.2
  volumeScaleAttackTime=0.2
  volumeScaleReleaseTime=0.2
 }
}
