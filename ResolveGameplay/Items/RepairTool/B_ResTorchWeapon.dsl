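; Fortnite repair torch weapon Blueprint that drives flame effects, audio, decals and force feedback.
; Equip, holster, fire and ammo events switch between idle, active and sputter-out effects, with a timer delay and shrink timeline.
; Impact gameplay cues color the flame and decals for healable (metal buildings, vehicles) or other targets, and choose surface sounds.
; Target health scales heal-loop volume, while a timer and RepNotify flags handle loop fades.
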(event Custom|OnHolstered
  ;[1 1 1] Holster/Unholster events to hide sounds
  (if (FortWeapon|IsEquipped)
    (Utilities|Time|ClearandInvalidateTimerbyHandle (Variables|Default|GetIdleTimer))
    (bind _sfx_idle_lp (Variables|SoundHookups|GetSFX_Idle_lp)
      (:then
        _)
      (:else
        _))
    (Audio|Components|Audio|FadeOut _sfx_idle_lp 0.2 0.0 "Linear" :!class /Script/Engine.AudioComponent)))

(event Custom|OnUnholstered
  ;[1 1 1] Holster/Unholster events to hide sounds
  (if (FortWeapon|IsEquipped)
    (bind _sfx_idle_lp (Variables|SoundHookups|GetSFX_Idle_lp)
      (:then
        (Utilities|FlowControl|Delay (Variables|Other|GetKnobTurnDelay))
        (Audio|Components|Audio|FadeIn _sfx_idle_lp 0.2 1.0 0.0 "Linear" :!class /Script/Engine.AudioComponent))
      (:else))))

(event EventBeginPlay
  ;[1 1 1] Holster/Unholster events to hide sounds
  (|CallParentFunction)
  (bind _asfort_player_pawn (Utilities|Casting|CastToFortPlayerPawn (Actor|GetOwner)))
  (EventDispatchers|BindEventtoOnHolsteredEvent _asfort_player_pawn (EventDispatchers|CreateEvent :!event Custom|OnHolstered))
  (EventDispatchers|BindEventtoOnUnholsteredEvent _asfort_player_pawn (EventDispatchers|CreateEvent :!event Custom|OnUnholstered)))

(event GameplayCue|GameplayCue.Impact.Generic
  (params (EventType :type EGameplayCueEvent%20Enum :default "OnActive") (Parameters :type Gameplay%20Cue%20Parameters%20Structure%20%28by%20ref%29))
  (bind _sfx_heal_lp (Variables|Default|GetSFX_Heal_LP))
  ;[0.6476 0.6476 0.6476] Gameplay cue on impact
  (bind _output_get (Variables|Default|SetImpactParams Parameters))
  (CallFunction|PlayImpactFXOnCue _output_get)
  (FortWeapon|PlayForceFeedback false)
  (Utilities|IsHealableTarget (.HitActor (Getters|GetActorandLocation _output_get))
    (:true
      (CallFunction|DestroyTimer)
      (bind _returnvalue (Utilities|Time|SetTimerbyEvent (EventDispatchers|CreateEvent :!event Custom|SetHealingOff) 0.3))
      (Variables|Default|SetTurnOffHealTimer _returnvalue)
      (if (not (Components|Activation|IsActive (Variables|Default|GetSFX_Heal_LP)))
        (Audio|Components|Audio|FadeIn _sfx_heal_lp 0.1 1.0 0.0 "Linear" :!class /Script/Engine.AudioComponent)))
    (:false
      (Utilities|Time|ClearandInvalidateTimerbyHandle (Variables|Default|GetTurnOffHealTimer))
      (Audio|Components|Audio|FadeOut _sfx_heal_lp 0.2 0.0 "Linear" :!class /Script/Engine.AudioComponent))))

(event Custom|SetHealingOff
  ;[1 1 1] Clear healing sound if no longer healing something
  (Variables|Default|SetOnHealLoopFaded true)
  (Utilities|Time|ClearandInvalidateTimerbyHandle (Variables|Default|GetTurnOffHealTimer)))

(event Custom|PlayImpactFXOnCue
  (params (InputParams :type Gameplay%20Cue%20Parameters%20Structure))
  (bind (_impact_sound _healthpercent) (Getters|GetSFXbyActor (.HitActor (Getters|GetActorandLocation InputParams))))
  (bind (_location _hitactor) (Getters|GetActorandLocation InputParams))
  (bind _returnvalue (Ability|GameplayCue|GetHitResult InputParams))
  (bind (_bblockinghit _binitialoverlap _time _distance _location_2 _impactpoint _normal) (Collision|BreakHitResult _returnvalue))
  (bind _torch_impact (Variables|Default|GetTorch%20Impact))
  ;[0.08601 0.3473 1] Burst effects each time this weapon hits a target (fires times 6/second)
  (Projectile|AddGameplayCue%28Looping%29Local _hitactor (Variables|Default|GetGameplay%20Cue%20Tag) (GameplayCue|MakeGameplayCueParametersfromHitResult _returnvalue))
  (bind _output_get (Variables|Default|SetImpact%20Normal _normal))
  (bind _output_get_2 (Variables|Other|SetImpactLocation _location))
  (bind _torch_impact_2 (Variables|Default|GetTorch%20Impact)
    (:then
      (Transformation|SetWorldTransform _torch_impact_2 (Math|Transform|MakeTransform _output_get_2 (Math|Conversions|RotationFromXVector _output_get) "0.250000,0.250000,0.250000") false true))
    (:else
      (bind _returnvalue_2 (Niagara|SpawnSystematLocation :SystemTemplate (Variables|ArtHookups|GetNS_Impact) :Location _output_get_2 :Scale "0.250000,0.250000,0.250000" :bAutoDestroy true :bAutoActivate false :PoolingMethod "None" :bPreCullCheck true))
      (Variables|Default|SetTorch%20Impact _returnvalue_2)))
  (Niagara|SetNiagaraVariable(Bool) _torch_impact "Healable?" (Getters|GetFXbyActor _hitactor))
  (CallFunction|PlayFXLatent _torch_impact)
  (Audio|PlaySoundatLocation _impact_sound (Variables|Other|GetImpactLocation))
  (if (>= _healthpercent 0)
    (bind _output_get_3 (Variables|Default|SetHealedPercent _healthpercent))
    (Audio|Parameter|SetFloatParameter (Variables|Default|GetSFX_Heal_LP) "Volume" (Math|Float|MapRangeClamped _output_get_3 0.0 1.0 0.2 1.0) :!class /Script/AudioExtensions.AudioParameterControllerInterface)
    (if (>= (Math|Float|MapRangeClamped _output_get_3 0.0 1.0 0.2 1.0) 1.0)
      (Utilities|FlowControl|DoOnce false :!label _label_2
        (:Completed
          (Audio|SpawnSoundatLocation (Variables|SoundHookups|GetSFX_FullyHealedOneShot) (Variables|Other|GetImpactLocation))
          (goto _label Reset))))
    (Utilities|FlowControl|DoOnce false :!label _label
      (:Completed
        (goto _label_2 Reset)))))

(event FortWeapon|EventOnEquippedAndAttached
  (bind _returnvalue (GameplayTags|MakeLiteralGameplayTagContainer "(GameplayTags=((TagName=\"Gameplay.Action.Player.AthenaInVehicle\"),(TagName=\"Gameplay.Action.Player.Ascender\")))"))
  (bind _instigator (Variables|Actor|GetInstigator))
  (bind _asfort_player_pawn (Utilities|Casting|CastToFortPlayerPawn _instigator))
  (bind _returnvalue_2 (GameplayTags|HasAnyMatchingGameplayTags _asfort_player_pawn _returnvalue))
  (bind _returnvalue_3 (select _returnvalue_2 "SnapToTargetIncludingScale" "SnapToTargetIncludingScale"))
  (bind _weaponmesh (Variables|Visual|GetWeaponMesh))
  ;[0.08601 0.3473 1] Equip / Unequip
  (|CallParentFunction)
  (bind _returnvalue_4 (Effects|SpawnFXSystemAttached :FXSystemAsset (Variables|ArtHookups|GetNS_Idle) :AttachToComponent _weaponmesh :AttachPointName "Muzzle" :Scale "1.000000,1.000000,1.000000" :LocationType _returnvalue_3 :bAutoDestroy false :PoolingMethod "None" :bAutoActivate false))
  (Variables|Default|SetFXComp_IdleFire _returnvalue_4)
  (bind _returnvalue_5 (Effects|SpawnFXSystemAttached :FXSystemAsset (Variables|ArtHookups|GetNS_RepairToolCore) :AttachToComponent _weaponmesh :AttachPointName "Muzzle" :Scale "1.000000,1.000000,1.000000" :LocationType _returnvalue_3 :bAutoDestroy false :PoolingMethod "None" :bAutoActivate false))
  (Variables|Default|SetFXComp_FireCore _returnvalue_5)
  ;;[0 0.1236 0.4323] Delay FX until pawn turns knob
  (bind _returnvalue_6 (Utilities|Time|SetTimerbyEvent (delegate Custom|ActivateIdleSound) (Variables|Other|GetKnobTurnDelay)))
  (Variables|Default|SetIdleTimer _returnvalue_6))

(event Custom|PlayFXLatent
  (params (NS :type Niagara%20Particle%20System%20Component%20Object%20Reference))
  ;[0.08601 0.3473 1] Delay FX when initially firing
  (Components|Activation|Activate NS true)
  (CallFunction|DrawDecalonTarget (Variables|Other|GetImpactLocation) (Variables|Default|GetTargetHealable?) (Variables|Default|GetImpact%20Normal)))

(event Custom|ResetLatentFX
  ;[0.08601 0.3473 1] Delay FX when initially firing
  (Utilities|DoOnceThenDo false
    (:DoOnce
      (Utilities|FlowControl|Delay 0.1)
      (Components|Activation|Activate (stale Custom|PlayFXLatent NS) true))
    (:ThenDo
      (Components|Activation|Activate (stale Custom|PlayFXLatent NS) false)))
  (CallFunction|DrawDecalonTarget (Variables|Other|GetImpactLocation) (Variables|Default|GetTargetHealable?) (Variables|Default|GetImpact%20Normal)))

(event Custom|ActivateIdleSound
  ;[0.08601 0.3473 1] Equip / Unequip
  ;;[0 0.1236 0.4323] Delay FX until pawn turns knob
  (Variables|Other|SetOnEffectsFadeIn true)
  (CallFunction|EnableAmmoEvents))

(event FortWeapon|EventOnLocalUnEquip
  ;[0.08601 0.3473 1] Equip / Unequip
  (|CallParentFunction)
  (CallFunction|RepUnequipFX)
  (CallFunction|DeactivateAudioComponents))

(event Custom|ShrinkAndFadeActiveFire
  (bind _fxcomp_firecore (Variables|Default|GetFXComp_FireCore))
  (bind _originalfirescale (Variables|Default|GetOriginalFireScale))
  ;[0.08601 0.3473 1] Timeline for scaling down Active Fire
  (bind (_direction _value) (Timeline|ScaleUpTimeline PlayFromStart)
    (:Update
      (Transformation|SetRelativeScale3D _fxcomp_firecore (Math|Vector|Lerp(Vector) _originalfirescale "0, 0, 0" _value)))
    (:Finished
      (Components|Activation|Deactivate _fxcomp_firecore)
      (Transformation|SetRelativeScale3D _fxcomp_firecore _originalfirescale))))

(event Custom|RepUnequipFX
  ;[0.08601 0.3473 1] Equip / Unequip
  (Utilities|Time|ClearandInvalidateTimerbyHandle (Variables|Default|GetIdleTimer))
  (Audio|Components|Audio|FadeOut (Variables|SoundHookups|GetSFX_Idle_lp) 0.2 0.0 "Linear" :!class /Script/Engine.AudioComponent)
  (Components|DestroyComponent (Variables|Default|GetFXComp_IdleFire))
  (Components|DestroyComponent (Variables|Default|GetFXComp_FireCore)))

(event Custom|StopTimeline
  (bind _fxcomp_firecore (Variables|Default|GetFXComp_FireCore))
  (bind _originalfirescale (Variables|Default|GetOriginalFireScale))
  ;[0.08601 0.3473 1] Timeline for scaling down Active Fire
  (Transformation|SetRelativeScale3D (Variables|Default|GetFXComp_FireCore) (Variables|Default|GetOriginalFireScale))
  (Timeline|ScaleUpTimeline Stop))

(event EventOnAmmoCountChanged
  (bind _returnvalue (FortWeapon|GetMagazineAmmoCount))
  (bind _minammotorestartfx (Variables|Other|GetMinAmmoToRestartFX))
  (bind _returnvalue_2 (>= _returnvalue _minammotorestartfx))
  (bind _fxcomp_idlefire (Variables|Default|GetFXComp_IdleFire))
  (bind _sfx_idle_lp (Variables|SoundHookups|GetSFX_Idle_lp))
  (bind _fxcomp_idlefire_2 (Variables|Default|GetFXComp_IdleFire))
  ;[0.08601 0.3473 1] Play smoke/disable idle flames if ammo is zero
  (Utilities|IsValid _fxcomp_idlefire_2
    (:Is%20Valid
      (if (<= _returnvalue 0)
        (Utilities|Time|ClearandInvalidateTimerbyHandle (Variables|Default|GetIdleTimer))
        (bind _returnvalue_3 (Effects|SpawnFXSystemAttached :FXSystemAsset (Variables|ArtHookups|GetNS_FlameEnd) :AttachToComponent (Variables|Visual|GetWeaponMesh) :AttachPointName "Muzzle" :Scale "1.000000,1.000000,1.000000" :LocationType "SnapToTargetIncludingScale" :bAutoDestroy true :PoolingMethod "None" :bAutoActivate false))
        (Setters|SetEndFlameColor _returnvalue_3)
        (Components|Activation|Activate _returnvalue_3 false)
        (Components|Activation|Deactivate _fxcomp_idlefire)
        (Audio|Components|Audio|FadeOut _sfx_idle_lp 0.2 0.0 "Linear" :!class /Script/Engine.AudioComponent)
        (Variables|Other|SetIsSputtering? true)
        (elif _returnvalue_2
          (Variables|Other|SetIsSputtering? false)
          (if (Math|Boolean|ANDBoolean _returnvalue_2 (not (Components|Activation|IsActive _fxcomp_idlefire_2)) (not (Components|Activation|IsActive (Variables|Default|GetFXComp_FireCore))))
            (Utilities|FlowControl|Gate true
              (:Exit
                (Components|Activation|Activate _fxcomp_idlefire true)
                (Audio|Components|Audio|FadeIn _sfx_idle_lp 0.3 1.0 0.0 "Linear" :!class /Script/Engine.AudioComponent)
                (Development|PrintString "Hello")))))))))

(event EventOnPlayWeaponFireFX
  (params (bPersistentFire :type Boolean) (bSecondaryFire :type Boolean))
  ;[0.08601 0.3473 1] Held Trigger Events & FX
  (CallFunction|PlayWeaponFireFXForceFeedback)
  (if (Components|Timeline|IsPlaying (Variables|B_ResTorchWeapon|GetScaleUpTimeline))
    (CallFunction|StopTimeline))
  (bind _sfx_idle_lp (Variables|SoundHookups|GetSFX_Idle_lp)
    (:then
      _)
    (:else
      _))
  (Audio|Components|Audio|FadeOut _sfx_idle_lp 0.3 0.0 "Linear" :!class /Script/Engine.AudioComponent)
  (if (not (Variables|Other|GetIsSputtering?))
    (bind _flamestartsound (Variables|SoundHookups|GetFlameStartSound)
      (:then
        (Audio|SpawnSoundAttached _flamestartsound (Variables|Visual|GetWeaponMesh))
        (Utilities|FlowControl|Delay 0.025)
        (Audio|Components|Audio|FadeIn (Variables|Default|GetSFX_Activated_lp) 0.1 1.0 0.0 "Linear" :!class /Script/Engine.AudioComponent)
        (SoundIndicator|StartSoundIndicatorLooping (Variables|SoundIndicator|GetSoundIndicatorComponent) (Transformation|GetActorLocation) _flamestartsound 1.0 0)
        (Components|Activation|Deactivate (Variables|Default|GetFXComp_IdleFire))
        (Components|Activation|Activate (Variables|Default|GetFXComp_FireCore) true))
      (:else))))

(event Custom|PlayWeaponFireFX_ForceFeedback
  ;[0.08601 0.3473 1] Play Fire Feedback Shake
  (bind _asfort_player_controller (Utilities|Casting|CastToFortPlayerController (Game|GetInstigatorController)))
  (Utilities|IsValid _asfort_player_controller
    (:Is%20Valid
      (if (Pawn|IsLocalController _asfort_player_controller)
        (Utilities|IsValid (Variables|Default|GetForceFeedbackComp_WeaponFire)
          (:Is%20Valid
            (ForceFeedback|Play (Variables|Default|GetForceFeedbackComp_WeaponFire) 0.0))
          (:Is%20Not%20Valid
            (bind _returnvalue (ForceFeedback|SpawnForceFeedbackAttached "/Game/Weapons/ForceFeedback/FF_Athena_Fire_Med_R.FF_Athena_Fire_Med_R" (Variables|Transformation|GetRootComponent))) ; spawning comp plays feedback shake
            (Variables|Default|SetForceFeedbackComp_WeaponFire _returnvalue)))))))

(event EventOnStopWeaponFireFX
  ;[0.08601 0.3473 1] Held Trigger Events & FX
  (bind _sfx_heal_lp (Variables|Default|GetSFX_Heal_LP)
    (:then
      (Audio|Components|Audio|FadeOut _sfx_heal_lp 0.2 0.0 "Linear" :!class /Script/Engine.AudioComponent)
      (bind _sfx_activated_lp (Variables|Default|GetSFX_Activated_lp)
        (:then
          (Audio|Components|Audio|FadeOut _sfx_activated_lp 0.1 0.0 "Linear" :!class /Script/Engine.AudioComponent)
          (bind _sfx_idle_lp (Variables|SoundHookups|GetSFX_Idle_lp)
            (:then
              (Audio|Components|Audio|FadeIn _sfx_idle_lp 0.2 1.0 0.0 "Linear" :!class /Script/Engine.AudioComponent)
              (bind _flamestopsound (Variables|SoundHookups|GetFlameStopSound)
                (:then
                  (Audio|SpawnSoundAttached _flamestopsound (Variables|Visual|GetWeaponMesh))
                  (SoundIndicator|StopSoundIndicatorLooping (Variables|SoundIndicator|GetSoundIndicatorComponent))
                  (Components|Activation|Activate (Variables|Default|GetFXComp_IdleFire) true)
                  (CallFunction|ShrinkandFadeActiveFire))
                (:else)))
            (:else)))
        (:else)))
    (:else)))

(event Custom|EnableAmmoEvents
  ;[0.08601 0.3473 1] Play smoke/disable idle flames if ammo is zero
  (Utilities|FlowControl|Gate true
    (:Exit
      (Components|Activation|Activate (Variables|Default|GetFXComp_IdleFire) true)
      (Audio|Components|Audio|FadeIn (Variables|SoundHookups|GetSFX_Idle_lp) 0.3 1.0 0.0 "Linear" :!class /Script/Engine.AudioComponent)
      (Development|PrintString "Hello"))))

(event Custom|DestroyTimer
  (bind _turnoffhealtimer (Variables|Default|GetTurnOffHealTimer))
  (if (Utilities|Time|IsValidTimerHandle _turnoffhealtimer)
    (Utilities|Time|ClearandInvalidateTimerbyHandle _turnoffhealtimer)))

(fn DeactivateAudioComponents
  (Utilities|FlowControl|Sequence
    (:then_0
      (bind _sfx_activated_lp (Variables|Default|GetSFX_Activated_lp)
        (:then
          (if (Audio|Components|Audio|IsPlaying _sfx_activated_lp)
            _))
        (:else
          _))
      (Audio|Components|Audio|Stop _sfx_activated_lp))
    (:then_1
      (bind _sfx_heal_lp (Variables|Default|GetSFX_Heal_LP)
        (:then
          (if (Audio|Components|Audio|IsPlaying _sfx_heal_lp)
            _))
        (:else
          _))
      (Audio|Components|Audio|Stop _sfx_heal_lp))))

(fn DrawDecalOnTarget
  (params (Location :type Vector) (Healing? :type Boolean) (Normal :type Vector))
  (Utilities|SpawnDecalSimple "5.000000,13.000000,13.000000" "6.000000,15.000000,15.000000" Location (CallFunction|MakeNormalizedRot Normal) 1.3 2.0 (select Healing? (Variables|CustomDecal|GetDecalMasterHeals) (Variables|CustomDecal|GetDecalMasterDamage)) "/Game/Effects/Fort_Effects/Gameplay/SmokeDustExplosions/Textures/T_GroundChar_01.T_GroundChar_01" (Math|Color|SelectColor (Variables|CustomDecal|GetDecalHealColor) (Variables|CustomDecal|GetDecalDamageColor) (Variables|Default|GetTargetHealable?)) (Math|Color|SelectColor (LocalVariables|GetDecalHealEndColor) (LocalVariables|GetDecal%20Damage%20End%20Color) (Variables|Default|GetTargetHealable?))))

(fn GetFXByActor
  (params (Actor :type Actor%20Object%20Reference))
  (bind _fxcomp_firecore (Variables|Default|GetFXComp_FireCore))
  (bind _asbuilding_actor (Utilities|Casting|CastToBuildingActor Actor)
    (:then
      (if (or (CallFunction|IsSMActorHealable? _asbuilding_actor) (GameplayTags|HasMatchingGameplayTag _asbuilding_actor "(TagName=\"Building.Resource.Metal\")"))
        (Effects|Components|ParticleSystem|SetColorParameter _fxcomp_firecore "TorchColor" "(R=0.000000,G=0.186874,B=5.000000,A=0.650000)" :!label _label_2)
        (Variables|Default|SetTargetHealable? true)
        (else
          (Effects|Components|ParticleSystem|SetColorParameter _fxcomp_firecore "TorchColor" "(R=5.000000,G=0.423301,B=0.000000,A=0.650000)" :!label _label)
          (Variables|Default|SetTargetHealable? false)))
      (return (Variables|Default|GetTargetHealable?)))
    (:CastFailed
      (if (not (Utilities|DoesObjectImplementInterface Actor "/Script/FortniteGame.FortVehicleInterface"))
        (goto _label))))
  (goto _label_2))

(fn GetSFXByActor
  (params (SurfacedActor :type Actor%20Object%20Reference))
  (bind (_asfort_vehicle_interface _bsuccess) (Utilities|Casting|CastToFortVehicleInterface SurfacedActor))
  (bind _sfx_impactgeneric (Variables|SoundHookups|GetSFX_ImpactGeneric))
  (bind _asbuilding_actor (Utilities|Casting|CastToBuildingActor SurfacedActor)
    (:then
      (LocalVariables|SetHealthPercentage (Math|Float|SelectFloat (Health|GetHealthPercent _asbuilding_actor :!class /Script/FortniteGame.BuildingActor) -1.0 (GameplayTags|HasMatchingGameplayTag _asbuilding_actor "(TagName=\"Building.Resource.Metal\")")))
      (LocalVariables|SetImpactSFX (select (GameplayTags|HasMatchingGameplayTag _asbuilding_actor "(TagName=\"Building.Resource.Wood\")") (Variables|SoundHookups|GetSFX_ImpactWood) (select (GameplayTags|HasMatchingGameplayTag _asbuilding_actor "(TagName=\"Building.Resource.Metal\")") (Variables|SoundHookups|GetSFX_ImpactMetal) _sfx_impactgeneric))))
    (:CastFailed
      (if _bsuccess
        (LocalVariables|SetHealthPercentage (Math|Float|SafeDivide (Class|FortVehicleInterface|GetHealth _asfort_vehicle_interface) (Class|FortVehicleInterface|GetMaxHealth _asfort_vehicle_interface)))
        (else
          (LocalVariables|SetHealthPercentage -1.0)))
      (LocalVariables|SetImpactSFX _sfx_impactgeneric)))
  (return (LocalVariables|GetImpactSFX) (LocalVariables|GetHealthPercentage)))

(fn GetActorAndLocation
  (params (Parameters :type Gameplay%20Cue%20Parameters%20Structure%20%28by%20ref%29))
  (bind (_normalizedmagnitude _rawmagnitude _effectcontext) (Ability|GameplayCue|BreakGameplayCueParameters Parameters))
  (bind (_bblockinghit _binitialoverlap _time _distance _location _impactpoint _normal _impactnormal _physmat _hitactor) (Collision|BreakHitResult (Ability|EffectContext|GetHitResult _effectcontext)))
  (return _location _hitactor))

(fn IsFullHealth?
  (params (Actor :type Actor%20Object%20Reference))
  (bind (_bsuccessfullyfoundattribute _returnvalue) (Ability|Attribute|GetFloatAttribute Actor "(AttributeName=\"Health\",Attribute=/Script/FortniteGame.FortHealthSet:Health,AttributeOwner=\"/Script/CoreUObject.Class'/Script/FortniteGame.FortHealthSet'\")"))
  (bind (_bsuccessfullyfoundattribute_2 _returnvalue_2) (Ability|Attribute|GetFloatAttribute Actor "(AttributeName=\"MaxHealth\",Attribute=/Script/FortniteGame.FortHealthSet:MaxHealth,AttributeOwner=\"/Script/CoreUObject.Class'/Script/FortniteGame.FortHealthSet'\")"))
  (return (Math|Float|NearlyEqual(Float) (Math|Float|SafeDivide _returnvalue _returnvalue_2) 1.0) (Math|Float|SafeDivide _returnvalue _returnvalue_2)))

(fn IsSMActorHealable?
  (params (Actor :type Building%20Actor%20Object%20Reference))
  (bind _asbuilding_static_mesh (Utilities|Casting|CastToBuildingSMActor Actor)
    (:then
      (if (Utilities|Enum|Equal(Enum) (Variables|/Script/FortniteGame.BuildingSMActor|GetResourceType _asbuilding_static_mesh) "Metal")
        (return true)
        (else
          _)))
    (:CastFailed
      _))
  (return false))

(fn MakeNormalizedRot
  (params (Normal :type Vector))
  (return (Math|Rotator|MakeRotationfromAxes Normal)))

(fn OnRep_OnEffectsFadeIn
  (if (Variables|Other|GetOnEffectsFadeIn)
    (if (and (> (FortWeapon|GetRemainingAmmo) 0) (not (GameplayTags|HasMatchingGameplayTag (Utilities|Casting|CastToFortPlayerPawn (Variables|Actor|GetInstigator)) "(TagName=\"Gameplay.Status.InVehicle\")")))
      (Components|Activation|Activate (Variables|Default|GetFXComp_IdleFire) false)
      (Audio|Components|Audio|FadeIn (Variables|SoundHookups|GetSFX_Idle_lp) 0.2 1.0 0.0 "Linear" :!class /Script/Engine.AudioComponent)
      (Variables|Other|SetOnEffectsFadeIn false)))
  (return))

(fn OnRep_OnHealLoopFaded
  (if (Variables|Default|GetOnHealLoopFaded)
    (Audio|Components|Audio|FadeOut (Variables|Default|GetSFX_Heal_LP) 0.25 0.0 "Linear" :!class /Script/Engine.AudioComponent)
    (Variables|Default|SetOnHealLoopFaded false)
    (Development|PrintString "fading out")))

(fn PlayHealingLoop
  (params (Actor :type Actor%20Object%20Reference))
  (bind _asbuilding_actor (Utilities|Casting|CastToBuildingActor Actor)
    (:then
      (if (not (GameplayTags|HasMatchingGameplayTag _asbuilding_actor "(TagName=\"Building.Resource.Metal\")"))
        (return)))
    (:CastFailed
      (if (not (Utilities|DoesObjectImplementInterface Actor "/Script/FortniteGame.FortVehicleInterface"))
        (return))))
  (Audio|Components|Audio|FadeIn (Variables|Default|GetSFX_Heal_LP) 0.3 1.0 0.0 "Linear" :!class /Script/Engine.AudioComponent))

(fn SetEndFlameColor
  (params (FXComp :type FXSystem%20Component%20Object%20Reference))
  (Effects|Components|ParticleSystem|SetColorParameter FXComp "TorchColor" (Math|Color|SelectColor (LocalVariables|GetHealColor) (LocalVariables|GetDamageColor) (Variables|Default|GetTargetHealable?)))
  (return))

(fn ConstructionScript)
