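# Niagara system for the repair tool's torch core: a glowing, flickering cone flame with flare sprites and a light.
# Four mesh cone layers re-burst on short loops, with curves indexed by Emitter.Age shaping an ignition flash and scale ramp.
# User.TorchColor tints both outer cones and the light-emitting FlickerFlare, while other layers use fixed colors.
# Dense alpha curves make both flares flicker; the Smoke and GPU outwardsparks emitters are disabled.
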NiagaraSystem /ResolveGameplay/Items/RepairTool/FX/NS_RepairToolCore
dependencies {
 usedDataInterfaces [
  &NiagaraDataInterfaceCurve
  &NiagaraDataInterfaceVectorCurve
  &NiagaraDataInterfaceMeshRendererInfo
  &NiagaraDataInterfaceVector2DCurve
  &NiagaraDataInterfaceArrayFloat
 ]
 usedDynamicInputs [&Vector2DFromFloat]
 usedModules [&SystemState, &Timeline]
}
emitters {
 CoreCone {
  emitterUpdateScript [
   &EmitterState(Life Cycle Mode=::NewEnumerator1, Loop Duration=0.1, UseLoopDelay=true)
   &SpawnBurst_Instantaneous(Use Spawn Probability=false, UseLoopCountLimit=false)
  ]
  particleSpawnScript [
   &InitializeParticle(
    Color=<rgba 0 13.798 15 1>
    Lifetime=0.25
    Mesh Scale Max=<0.15 0.45 0.45>
    Mesh Scale Min=<0.1 0.4 0.4>
    Mesh Scale Mode=::NewEnumerator2
    Position Offset=<2 0 0>
    UsePositionOffset=true
    Write Material Random=true
   ) {module=InitializeParticle001}
   &InitialMeshOrientation(Mesh Orientation Mode=::NewEnumerator7, Rotation=<0 0.5 0>, Use Rotation Vector=false)
  ]
  particleUpdateScript [
   &ParticleState()
   &ScaleColor(
    Scale Alpha=&FloatFromCurve(FloatCurve=DataInterface() {propertyValues {Curve {keys [<at 0 0 RCIM_Cubic>, {arriveTangent=0.75, interpMode=RCIM_Cubic, leaveTangent=0.75, time=0.5, value=1}, <at 1 0.75 RCIM_Cubic>]}}})
    Scale RGB=&Multiply_VectorByFloat(
     Vector=<1 1 1>
     Float=&Add_Float(
      A=1
      B=&FloatFromCurve(
       FloatCurve=DataInterface() {propertyValues {Curve {keys [
           {arriveTangent=9.0147, interpMode=RCIM_Cubic, leaveTangent=9.0147, tangentMode=RCTM_User, time=0, value=0}
           {arriveTangent=-3.675, interpMode=RCIM_Cubic, leaveTangent=-3.675, tangentMode=RCTM_User, time=0.2, value=1}
           {arriveTangent=-0.0405, interpMode=RCIM_Cubic, leaveTangent=-0.0405, tangentMode=RCTM_User, time=1, value=0}
          ]}}}
       CurveIndex=$Emitter.Age
       Scale Curve=5
      )
     )
    )
    ScaleA=true
    ScaleRGB=true
   )
   &DynamicMaterialParameters(
    Index 0 Param 2=0.5
    Index 0 Param 3=15
    Index 0 Param 4=1.5
    Module.Param0WriteEnabled=true
    Module.Param1WriteEnabled=true
    Module.Param2WriteEnabled=true
    Module.Param3WriteEnabled=true
   )
   &ScaleMeshSize(Scale Factor=&VectorFromCurve(
     VectorCurve=DataInterface() {propertyValues {
       XCurve {keys [<at 0 0.75 RCIM_Cubic>,<at 1 2 RCIM_Cubic>]}
       YCurve {keys [<at 0 1 RCIM_Cubic>,<at 1 0.9 RCIM_Cubic>]}
       ZCurve {keys [<at 0 1 RCIM_Cubic>,<at 1 0.9 RCIM_Cubic>]}
      }}
     Scale Curve=<0.5 0.25 0.25>
    )) {enabled=false}
   &ScaleMeshSize(Scale Factor=&VectorFromCurve(
     VectorCurve=DataInterface() {propertyValues {
       XCurve {keys [
         {arriveTangent=9.0725, interpMode=RCIM_Cubic, leaveTangent=9.0725, tangentMode=RCTM_User, time=0, value=0}
         {arriveTangent=-2.5249, interpMode=RCIM_Cubic, leaveTangent=-2.5249, tangentMode=RCTM_User, time=0.1408, value=0.9847}
         {arriveTangent=-0.0405, interpMode=RCIM_Cubic, leaveTangent=-0.0405, tangentMode=RCTM_User, time=1, value=0.5}
        ]}
       YCurve {keys [
         {arriveTangent=9.0725, interpMode=RCIM_Cubic, leaveTangent=9.0725, tangentMode=RCTM_User, time=0, value=0}
         {arriveTangent=-2.5249, interpMode=RCIM_Cubic, leaveTangent=-2.5249, tangentMode=RCTM_User, time=0.1408, value=0.9847}
         {arriveTangent=-0.0405, interpMode=RCIM_Cubic, leaveTangent=-0.0405, tangentMode=RCTM_User, time=1, value=0.5}
        ]}
       ZCurve {keys [
         {arriveTangent=9.0725, interpMode=RCIM_Cubic, leaveTangent=9.0725, tangentMode=RCTM_User, time=0, value=0}
         {arriveTangent=-2.5249, interpMode=RCIM_Cubic, leaveTangent=-2.5249, tangentMode=RCTM_User, time=0.1408, value=0.9847}
         {arriveTangent=-0.0405, interpMode=RCIM_Cubic, leaveTangent=-0.0405, tangentMode=RCTM_User, time=1, value=0.5}
        ]}
      }}
     CurveIndex=$Emitter.Age
     Scale Curve=<4 0.35 0.35>
    )) {module=ScaleMeshSize001}
  ]
  rendererClasses [&NiagaraMeshRendererProperties]
  renderers [{rendererClass=&NiagaraMeshRendererProperties, rendererIndex=0}]
 }
 CoreCone_Outer {
  emitterUpdateScript [
   &EmitterState(Inactive Response=::NewEnumerator1, Life Cycle Mode=::NewEnumerator1, Loop Duration=0.1, UseLoopDelay=true)
   &SpawnBurst_Instantaneous(Use Spawn Probability=false, UseLoopCountLimit=false)
  ]
  particleSpawnScript [
   &InitializeParticle(
    Color=$User.TorchColor
    Lifetime=0.25
    Mesh Scale Max=<0.18 0.45 0.45>
    Mesh Scale Min=<0.09 0.4 0.4>
    Mesh Scale Mode=::NewEnumerator2
    Position Offset=<2 0 0>
    UsePositionOffset=true
    Write Material Random=true
   ) {module=InitializeParticle001}
   &InitialMeshOrientation(Mesh Orientation Mode=::NewEnumerator7, Rotation=<0 0.5 0>, Use Rotation Vector=true)
  ]
  particleUpdateScript [
   &ParticleState()
   &ScaleColor(
    Scale Alpha=&FloatFromCurve(FloatCurve=DataInterface() {propertyValues {Curve {keys [<at 0 0 RCIM_Cubic>, {arriveTangent=0.45, interpMode=RCIM_Cubic, leaveTangent=0.45, time=0.5, value=1}, <at 1 0.45 RCIM_Cubic>]}}})
    Scale RGB=&Multiply_VectorByFloat(
     Vector=<1 1 1>
     Float=&Add_Float(
      A=1
      B=&FloatFromCurve(
       FloatCurve=DataInterface() {propertyValues {Curve {keys [
           {arriveTangent=9.0147, interpMode=RCIM_Cubic, leaveTangent=9.0147, tangentMode=RCTM_User, time=0, value=0}
           {arriveTangent=-3.675, interpMode=RCIM_Cubic, leaveTangent=-3.675, tangentMode=RCTM_User, time=0.2, value=1}
           {arriveTangent=-0.0405, interpMode=RCIM_Cubic, leaveTangent=-0.0405, tangentMode=RCTM_User, time=1, value=0}
          ]}}}
       CurveIndex=$Emitter.Age
       Scale Curve=5
      )
     )
    )
    ScaleA=true
    ScaleRGB=true
   )
   &DynamicMaterialParameters(
    Index 0 Param 2=1
    Index 0 Param 3=15
    Index 0 Param 4=&FloatFromCurve(
     FloatCurve=DataInterface() {propertyValues {Curve {keys [
         {interpMode=RCIM_Cubic, tangentMode=RCTM_User, time=0.1, value=0}
         {interpMode=RCIM_Cubic, tangentMode=RCTM_User, time=0.2, value=1.25}
         {tangentMode=RCTM_User, time=1, value=0.65}
        ]}}}
     CurveIndex=$Emitter.Age
     Scale Curve=4.3
    )
    Module.Param0WriteEnabled=true
    Module.Param1WriteEnabled=true
    Module.Param2WriteEnabled=true
    Module.Param3WriteEnabled=true
   )
   &ScaleMeshSize(Scale Factor=&VectorFromCurve(
     VectorCurve=DataInterface() {propertyValues {
       XCurve {keys [{interpMode=RCIM_Cubic, tangentMode=RCTM_User, time=0.1, value=0}, {interpMode=RCIM_Cubic, tangentMode=RCTM_User, time=0.2, value=1}]}
       YCurve {keys [{interpMode=RCIM_Cubic, tangentMode=RCTM_User, time=0.1, value=0}, {interpMode=RCIM_Cubic, tangentMode=RCTM_User, time=0.2, value=1}]}
       ZCurve {keys [{interpMode=RCIM_Cubic, tangentMode=RCTM_User, time=0.1, value=0}, {interpMode=RCIM_Cubic, tangentMode=RCTM_User, time=0.2, value=1}]}
      }}
     CurveIndex=$Emitter.Age
     Scale Curve=<6.5 0.33 0.33>
    )) {module=ScaleMeshSize001}
  ]
  rendererClasses [&NiagaraMeshRendererProperties]
  renderers [{rendererClass=&NiagaraMeshRendererProperties, rendererIndex=0}]
 }
 CoreCone_Outer001 {
  emitterUpdateScript [
   &EmitterState(Inactive Response=::NewEnumerator1, Life Cycle Mode=::NewEnumerator1, Loop Duration=0.1, UseLoopDelay=true)
   &SpawnBurst_Instantaneous(Use Spawn Probability=false, UseLoopCountLimit=false)
  ]
  particleSpawnScript [
   &InitializeParticle(
    Color=&MultiplyLinearColor(
     Linear Color=$User.TorchColor
     Scale Factor=&MakeVector4_FromFloat(Float=&FloatFromCurve(FloatCurve=DataInterface() {propertyValues {Curve {keys [<at 0 1>,<at 1 0>]}}}))
    )
    Lifetime=0.25
    Mesh Scale Max=<0.18 0.45 0.45>
    Mesh Scale Min=<0.09 0.4 0.4>
    Mesh Scale Mode=::NewEnumerator2
    Position Offset=<2 0 0>
    UsePositionOffset=true
    Write Material Random=true
   ) {module=InitializeParticle001}
   &InitialMeshOrientation(Mesh Orientation Mode=::NewEnumerator7, Use Rotation Vector=true)
  ]
  particleUpdateScript [
   &ParticleState()
   &ScaleColor(
    Scale Alpha=&FloatFromCurve(FloatCurve=DataInterface() {propertyValues {Curve {keys [<at 0 0 RCIM_Cubic>, {arriveTangent=0.45, interpMode=RCIM_Cubic, leaveTangent=0.45, time=0.5, value=1}, <at 1 0.45 RCIM_Cubic>]}}})
    Scale RGB=&Multiply_VectorByFloat(
     Vector=<1 1 1>
     Float=&Add_Float(
      A=1
      B=&FloatFromCurve(
       FloatCurve=DataInterface() {propertyValues {Curve {keys [
           {arriveTangent=9.0147, interpMode=RCIM_Cubic, leaveTangent=9.0147, tangentMode=RCTM_User, time=0, value=0}
           {arriveTangent=-3.675, interpMode=RCIM_Cubic, leaveTangent=-3.675, tangentMode=RCTM_User, time=0.3, value=0.75}
           {arriveTangent=-0.0405, interpMode=RCIM_Cubic, leaveTangent=-0.0405, tangentMode=RCTM_User, time=1, value=0.35}
          ]}}}
       CurveIndex=$Emitter.Age
       Scale Curve=5
      )
     )
    )
    ScaleA=true
    ScaleRGB=true
   )
   &DynamicMaterialParameters(
    Index 0 Param 2=1.5
    Index 0 Param 3=15
    Index 0 Param 4=&FloatFromCurve(
     FloatCurve=DataInterface() {propertyValues {Curve {keys [
         {interpMode=RCIM_Cubic, tangentMode=RCTM_User, time=0.1, value=0.05}
         {interpMode=RCIM_Cubic, tangentMode=RCTM_User, time=0.3, value=0.85}
         {tangentMode=RCTM_User, time=1, value=0.5}
        ]}}}
     CurveIndex=$Emitter.Age
     Scale Curve=4.2
    )
    Module.Param0WriteEnabled=true
    Module.Param1WriteEnabled=true
    Module.Param2WriteEnabled=true
    Module.Param3WriteEnabled=true
   )
   &ScaleMeshSize(Scale Factor=&VectorFromCurve(
     VectorCurve=DataInterface() {propertyValues {
       XCurve {keys [{interpMode=RCIM_Cubic, tangentMode=RCTM_User, time=0.1, value=-0.05}, {interpMode=RCIM_Cubic, tangentMode=RCTM_User, time=0.2, value=1}]}
       YCurve {keys [{interpMode=RCIM_Cubic, tangentMode=RCTM_User, time=0.1, value=-0.05}, {interpMode=RCIM_Cubic, tangentMode=RCTM_User, time=0.2, value=1}]}
       ZCurve {keys [{interpMode=RCIM_Cubic, tangentMode=RCTM_User, time=0.1, value=-0.05}, {interpMode=RCIM_Cubic, tangentMode=RCTM_User, time=0.2, value=1}]}
      }}
     CurveIndex=$Emitter.Age
     Scale Curve=<5 0.3 0.3>
    ))
  ]
  rendererClasses [&NiagaraMeshRendererProperties]
  renderers [{rendererClass=&NiagaraMeshRendererProperties, rendererIndex=0}]
 }
 FlickerCore {
  emitterUpdateScript [
   &EmitterState(Inactive Response=::NewEnumerator1, Life Cycle Mode=::NewEnumerator1, Loop Duration=0.1, UseLoopDelay=true)
   &SpawnBurst_Instantaneous(Use Spawn Probability=false, UseLoopCountLimit=false)
  ]
  particleSpawnScript [
   &InitializeParticle(
    Color=<rgba 10.9773 5.7601 15 1>
    Lifetime=0.25
    Mesh Scale Max=<0.35 0.23 0.23>
    Mesh Scale Min=<0.2 0.2 0.2>
    Mesh Scale Mode=::NewEnumerator2
    Position Offset=<2 0 0>
    UsePositionOffset=true
    Write Material Random=true
   ) {module=InitializeParticle001}
   &InitialMeshOrientation(Mesh Orientation Mode=::NewEnumerator7, Rotation=<0 0.5 0>, Use Rotation Vector=false)
  ]
  particleUpdateScript [
   &ParticleState()
   &ScaleColor(
    Scale Alpha=&FloatFromCurve(FloatCurve=DataInterface() {propertyValues {Curve {keys [<at 0 0 RCIM_Cubic>,<at 0.5 1 RCIM_Cubic>,<at 1 0 RCIM_Cubic>]}}})
    Scale RGB=&Multiply_VectorByFloat(
     Vector=<1 1 1>
     Float=&Add_Float(
      A=1
      B=&FloatFromCurve(
       FloatCurve=DataInterface() {propertyValues {Curve {keys [
           {arriveTangent=9.0147, interpMode=RCIM_Cubic, leaveTangent=9.0147, tangentMode=RCTM_User, time=0, value=0}
           {arriveTangent=-3.675, interpMode=RCIM_Cubic, leaveTangent=-3.675, tangentMode=RCTM_User, time=0.1, value=1}
           {arriveTangent=-0.0405, interpMode=RCIM_Cubic, leaveTangent=-0.0405, tangentMode=RCTM_User, time=1, value=0}
          ]}}}
       CurveIndex=$Emitter.Age
       Scale Curve=5
      )
     )
    )
    ScaleA=true
    ScaleRGB=true
   )
   &DynamicMaterialParameters(
    Index 0 Param 3=15
    Index 0 Param 4=1.5
    Module.Param0WriteEnabled=true
    Module.Param1WriteEnabled=true
    Module.Param2WriteEnabled=true
    Module.Param3WriteEnabled=true
   )
   &ScaleMeshSize(Scale Factor=&VectorFromCurve(
     VectorCurve=DataInterface() {propertyValues {
       XCurve {keys [
         <at 0 0.75 RCIM_Cubic>
         {arriveTangent=3.1603, interpMode=RCIM_Cubic, leaveTangent=3.1603, tangentMode=RCTM_User, time=0.8, value=2.125}
         {arriveTangent=-0.0874, interpMode=RCIM_Cubic, leaveTangent=-0.0874, tangentMode=RCTM_User, time=2, value=2.5}
        ]}
       YCurve {keys [<at 0 1 RCIM_Cubic>,<at 1 0.9 RCIM_Cubic>]}
       ZCurve {keys [<at 0 1 RCIM_Cubic>,<at 1 0.9 RCIM_Cubic>]}
      }}
     CurveIndex=$Emitter.Age
     Scale Curve=<1 0.4 0.4>
    ))
  ]
  rendererClasses [&NiagaraMeshRendererProperties]
  renderers [{rendererClass=&NiagaraMeshRendererProperties, rendererIndex=0}]
 }
 FlickerFlare {
  emitterUpdateScript [
   &EmitterState(Inactive Response=::NewEnumerator2, Life Cycle Mode=::NewEnumerator1, Loop Duration=0.15, UseLoopDelay=false)
   &SpawnBurst_Instantaneous(Use Spawn Probability=false, UseLoopCountLimit=false)
   &SpawnRate(SpawnRate=12, Use Spawn Probability=false)
  ]
  particleSpawnScript [&InitializeParticle(
    Color=$User.TorchColor
    Lifetime=0.25
    Position Offset=<2 0.25 0>
    Sprite Rotation Mode=::NewEnumerator2
    Sprite Size Mode=::NewEnumerator1
    Uniform Sprite Size Max=25
    UsePositionOffset=true
    Write Material Random=true
   ) {module=InitializeParticle001}]
  particleUpdateScript [
   &ParticleState()
   &DynamicMaterialParameters(
    Index 0 Param 1=0.3
    Index 0 Param 2=0.4
    Index 0 Param 3=1
    Index 0 Param 4=1
    Module.Param0WriteEnabled=true
    Module.Param1WriteEnabled=true
    Module.Param2WriteEnabled=true
    Module.Param3WriteEnabled=true
   ) {enabled=false}
   &ScaleSpriteSize(
    Uniform Curve Index=$Emitter.Age
    Uniform Curve Sprite Scale=DataInterface() {propertyValues {Curve {keys [{interpMode=RCIM_Cubic, tangentMode=RCTM_User, time=0, value=0}, {interpMode=RCIM_Cubic, tangentMode=RCTM_User, time=0.4, value=1}]}}}
   ) {module=ScaleSpriteSize001}
   &ScaleColor(
    Scale Alpha=&FloatFromCurve(
     FloatCurve=DataInterface() {propertyValues {Curve {keys [
         {arriveTangent=0.0922, time=0.0136, value=1.675}
         {arriveTangent=-0.4156, time=0.0205, value=1.975}
         {arriveTangent=0.7361, leaveTangent=0.7361, time=0.0799, value=0.7178}
         {arriveTangent=-0.4488, leaveTangent=-0.4488, time=0.0959, value=1.4539}
         {arriveTangent=0.1953, leaveTangent=0.1953, time=0.1058, value=1.0051}
         {arriveTangent=-0.5333, leaveTangent=-0.5333, time=0.123, value=1.2005}
         {arriveTangent=0.5413, leaveTangent=0.5413, time=0.1501, value=0.6671}
         {arriveTangent=-0.4256, leaveTangent=-0.4256, time=0.1833, value=1.2084}
         {arriveTangent=0.3174, leaveTangent=0.3174, time=0.235, value=0.7828}
         {arriveTangent=-0.3004, leaveTangent=-0.3004, time=0.2879, value=1.1003}
         {arriveTangent=0.3663, leaveTangent=0.3663, time=0.3175, value=0.7998}
         {arriveTangent=-0.2435, leaveTangent=-0.2435, time=0.3433, value=1.1661}
         {arriveTangent=0.1126, leaveTangent=0.1126, time=0.3729, value=0.9227}
         {arriveTangent=-0.1887, leaveTangent=-0.1887, time=0.4135, value=1.0353}
         {arriveTangent=0.7458, leaveTangent=0.7458, time=0.4467, value=0.8466}
         {arriveTangent=-0.6407, leaveTangent=-0.6407, time=0.4643, value=1.5923}
         {arriveTangent=0.1597, leaveTangent=0.1597, time=0.4836, value=0.9516}
         {arriveTangent=-0.1671, leaveTangent=-0.1671, time=0.523, value=1.1114}
         {arriveTangent=0.1638, leaveTangent=0.1638, time=0.5406, value=0.9443}
         {arriveTangent=-0.3391, leaveTangent=-0.3391, time=0.5615, value=1.1082}
         {arriveTangent=0.1403, leaveTangent=0.1403, time=0.5821, value=0.769}
         {arriveTangent=-0.155, leaveTangent=-0.155, time=0.5974, value=0.9093}
         {arriveTangent=0.4103, leaveTangent=0.4103, time=0.6182, value=0.7543}
         {arriveTangent=-0.2683, leaveTangent=-0.2683, time=0.6504, value=1.1646}
         {arriveTangent=-0.0333, leaveTangent=-0.0333, time=0.6844, value=0.8963}
         {arriveTangent=0.1953, leaveTangent=0.1953, time=0.7128, value=0.863}
         {arriveTangent=0.3263, leaveTangent=0.3263, time=0.7326, value=1.0583}
         {arriveTangent=-0.5488, leaveTangent=-0.5488, time=0.7496, value=1.3845}
         {arriveTangent=0.2314, leaveTangent=0.2314, time=0.7619, value=0.8357}
         {arriveTangent=-0.3297, leaveTangent=-0.3297, time=0.8073, value=1.0671}
         {arriveTangent=0.0791, leaveTangent=0.0791, time=0.8215, value=0.7375}
         {arriveTangent=-0.2574, leaveTangent=-0.2574, time=0.8524, value=0.8165}
         {arriveTangent=0.3961, leaveTangent=0.3961, time=0.8668, value=0.5591}
         {arriveTangent=-0.1383, leaveTangent=-0.1383, time=0.8846, value=0.9552}
         {arriveTangent=0.3356, leaveTangent=0.3356, time=0.9017, value=0.8169}
         {arriveTangent=-0.3459, leaveTangent=-0.3459, time=0.9349, value=1.1524}
         {arriveTangent=0.2311, leaveTangent=0.2311, time=0.9557, value=0.8065}
         {arriveTangent=-0.5769, leaveTangent=-0.5769, time=0.9774, value=1.0376}
         <at 1 0.4607>
        ]}}}
     Scale Curve=0.5
    )
    Scale RGB=&Multiply_VectorByFloat(
     Vector=<1 1 1>
     Float=&Add_Float(
      A=1
      B=&FloatFromCurve(
       FloatCurve=DataInterface() {propertyValues {Curve {keys [
           {arriveTangent=9.0147, interpMode=RCIM_Cubic, leaveTangent=9.0147, tangentMode=RCTM_User, time=0, value=0}
           {arriveTangent=-3.675, interpMode=RCIM_Cubic, leaveTangent=-3.675, tangentMode=RCTM_User, time=0.1, value=1}
           {arriveTangent=-0.0405, interpMode=RCIM_Cubic, leaveTangent=-0.0405, tangentMode=RCTM_User, time=1, value=0}
          ]}}}
       CurveIndex=$Emitter.Age
       Scale Curve=5
      )
     )
    )
    ScaleA=true
    ScaleRGB=true
   )
   &FN_Light_Attributes(Create Exponent Attribute=true, Exponent=4)
  ]
  rendererClasses [&NiagaraSpriteRendererProperties, &NiagaraLightRendererProperties]
  renderers [
   {rendererClass=&NiagaraSpriteRendererProperties}
   {rendererClass=&NiagaraLightRendererProperties, rendererIndex=1}
  ]
 }
 FlickerFlareWide {
  emitterUpdateScript [
   &EmitterState(Inactive Response=::NewEnumerator2, Life Cycle Mode=::NewEnumerator1, Loop Duration=0.15, UseLoopDelay=true)
   &SpawnBurst_Instantaneous(Use Spawn Probability=false, UseLoopCountLimit=false)
  ]
  particleSpawnScript [&InitializeParticle(
    Color=<rgba 0.5853 1 0.9199 0.5>
    Position Offset=<2 0 0>
    Sprite Size Max=<35 5>
    Sprite Size Min=<10 1>
    Sprite Size Mode=::NewEnumerator2
    UsePositionOffset=true
    Write Material Random=true
   ) {module=InitializeParticle001}]
  particleUpdateScript [
   &ParticleState()
   &DynamicMaterialParameters(
    Index 0 Param 1=0.1
    Index 0 Param 2=0.4
    Index 0 Param 3=1
    Index 0 Param 4=1
    Module.Param0WriteEnabled=true
    Module.Param1WriteEnabled=true
    Module.Param2WriteEnabled=true
    Module.Param3WriteEnabled=true
   )
   &ScaleColor(
    Scale Alpha=&FloatFromCurve(FloatCurve=DataInterface() {propertyValues {Curve {keys [
         {arriveTangent=1.0412, time=0, value=0}
         {arriveTangent=0.0922, leaveTangent=0.0922, time=0.0246, value=1.0412}
         {arriveTangent=-0.4156, leaveTangent=-0.4156, time=0.0561, value=1.1334}
         {arriveTangent=0.7361, leaveTangent=0.7361, time=0.0799, value=0.7178}
         {arriveTangent=-0.4488, leaveTangent=-0.4488, time=0.0959, value=1.4539}
         {arriveTangent=0.1953, leaveTangent=0.1953, time=0.1058, value=1.0051}
         {arriveTangent=-0.5333, leaveTangent=-0.5333, time=0.123, value=1.2005}
         {arriveTangent=0.5413, leaveTangent=0.5413, time=0.1501, value=0.6671}
         {arriveTangent=-0.4256, leaveTangent=-0.4256, time=0.1833, value=1.2084}
         {arriveTangent=0.3174, leaveTangent=0.3174, time=0.235, value=0.7828}
         {arriveTangent=-0.3004, leaveTangent=-0.3004, time=0.2879, value=1.1003}
         {arriveTangent=0.3663, leaveTangent=0.3663, time=0.3175, value=0.7998}
         {arriveTangent=-0.2435, leaveTangent=-0.2435, time=0.3433, value=1.1661}
         {arriveTangent=0.1126, leaveTangent=0.1126, time=0.3729, value=0.9227}
         {arriveTangent=-0.1887, leaveTangent=-0.1887, time=0.4135, value=1.0353}
         {arriveTangent=0.7458, leaveTangent=0.7458, time=0.4467, value=0.8466}
         {arriveTangent=-0.6407, leaveTangent=-0.6407, time=0.4643, value=1.5923}
         {arriveTangent=0.1597, leaveTangent=0.1597, time=0.4836, value=0.9516}
         {arriveTangent=-0.1671, leaveTangent=-0.1671, time=0.523, value=1.1114}
         {arriveTangent=0.1638, leaveTangent=0.1638, time=0.5406, value=0.9443}
         {arriveTangent=-0.3391, leaveTangent=-0.3391, time=0.5615, value=1.1082}
         {arriveTangent=0.1403, leaveTangent=0.1403, time=0.5821, value=0.769}
         {arriveTangent=-0.155, leaveTangent=-0.155, time=0.5974, value=0.9093}
         {arriveTangent=0.4103, leaveTangent=0.4103, time=0.6182, value=0.7543}
         {arriveTangent=-0.2683, leaveTangent=-0.2683, time=0.6504, value=1.1646}
         {arriveTangent=-0.0333, leaveTangent=-0.0333, time=0.6844, value=0.8963}
         {arriveTangent=0.1953, leaveTangent=0.1953, time=0.7128, value=0.863}
         {arriveTangent=0.3263, leaveTangent=0.3263, time=0.7326, value=1.0583}
         {arriveTangent=-0.5488, leaveTangent=-0.5488, time=0.7496, value=1.3845}
         {arriveTangent=0.2314, leaveTangent=0.2314, time=0.7619, value=0.8357}
         {arriveTangent=-0.3297, leaveTangent=-0.3297, time=0.8073, value=1.0671}
         {arriveTangent=0.0791, leaveTangent=0.0791, time=0.8215, value=0.7375}
         {arriveTangent=-0.2574, leaveTangent=-0.2574, time=0.8524, value=0.8165}
         {arriveTangent=0.3961, leaveTangent=0.3961, time=0.8668, value=0.5591}
         {arriveTangent=-0.1383, leaveTangent=-0.1383, time=0.8846, value=0.9552}
         {arriveTangent=0.3356, leaveTangent=0.3356, time=0.9017, value=0.8169}
         {arriveTangent=-0.3459, leaveTangent=-0.3459, time=0.9349, value=1.1524}
         {arriveTangent=0.2311, leaveTangent=0.2311, time=0.9557, value=0.8065}
         {arriveTangent=-0.5769, leaveTangent=-0.5769, time=0.9774, value=1.0376}
         <at 1 0.4607>
        ]}}})
    Scale RGB=&VectorFromCurve(VectorCurve=DataInterface() {propertyValues {XCurve {keys [<at 0 1>,<at 1 1>]}, YCurve {keys [<at 0 1>,<at 1 1>]}, ZCurve {keys [<at 0 1>,<at 1 1>]}}})
    ScaleA=true
    ScaleRGB=true
   )
   &ScaleSpriteSize(
    Uniform Curve Scale=0.25
    Uniform Curve Sprite Scale=DataInterface() {propertyValues {Curve {keys [
        {interpMode=RCIM_Cubic, tangentMode=RCTM_User, time=0, value=0}
        {interpMode=RCIM_Cubic, tangentMode=RCTM_User, time=0.5, value=1}
        {tangentMode=RCTM_User, time=1, value=0.45}
       ]}}}
   ) {enabled=false}
  ]
  rendererClasses [&NiagaraSpriteRendererProperties]
  renderers [{rendererClass=&NiagaraSpriteRendererProperties}]
 }
 Smoke {
  bEnabled=false
  emitterUpdateScript [
   &EmitterState()
   &SpawnBurst_Instantaneous(
    Spawn Count=&RandomRangeInt(Minimum=1, Maximum=2)
    Use Spawn Probability=false
    UseLoopCountLimit=false
   ) {enabled=false}
   &SpawnRate(Spawn Probability=0.4, SpawnRate=15, Use Spawn Probability=true)
  ]
  particleSpawnScript [
   &InitializeParticle(
    Color=<rgba 0.2135 0.212 0.2073 0.5>
    Lifetime Max=1
    Lifetime Min=0.5
    Lifetime Mode=::NewEnumerator1
    Position Offset=<-10 0 0>
    Sprite Rotation Mode=::NewEnumerator2
    Sprite Size Mode=::NewEnumerator1
    Uniform Sprite Size Max=20
    Uniform Sprite Size Min=7
    UsePositionOffset=false
    Write Material Random=true
   ) {module=InitializeParticle001}
   &ShapeLocation(
    Handle Radius=1
    Large Radius=4
    MaskSpawnGroup=false
    Override Seed=false
    Rotation Mode=::NewEnumerator1
    Shape Primitive=::NewEnumerator3
    UseOwnerScale=true
   ) {enabled=false}
   &AddVelocityFromPoint(
    Module.Use Falloff Distance=false
    Origin Offset=<-0.2 0 0>
    Velocity Strength=&RandomRangeFloat(Minimum=35, Maximum=85)
   ) {enabled=false}
   &AddVelocityInCone(
    Cone Angle=5
    Cone Axis=<0.5 0 0>
    Module.Use Velocity Falloff On Cone Axis=true
    Velocity Distribution Along Cone Axis=0.75
    Velocity Falloff Away From Cone Axis=0.333
    Velocity Strength=&RandomRangeFloat(Minimum=250, Maximum=500)
   )
   &SystemLocation(MaskSpawnGroup=false, Offset=<20 0 0>)
  ]
  particleUpdateScript [
   &ParticleState()
   &CurlNoiseForce(Noise Strength=300, Pan Noise=false) {enabled=false}
   &SolveForcesAndVelocity(Clamp Velocity=false, Limit Acceleration=false)
   &SubUVAnimation() {module=SubUVAnimation001}
   &DynamicMaterialParameters(
    Index 0 Param 1=&FloatFromCurve(FloatCurve=DataInterface() {propertyValues {Curve {keys [{interpMode=RCIM_Cubic, tangentMode=RCTM_User, time=0, value=0}, {interpMode=RCIM_Cubic, tangentMode=RCTM_User, time=1, value=1}]}}})
    Index 0 Param 2=1
    Index 0 Param 3=1
    Index 0 Param 4=1
    Module.Param0WriteEnabled=true
    Module.Param1WriteEnabled=true
    Module.Param2WriteEnabled=true
    Module.Param3WriteEnabled=true
   )
   &ScaleSpriteSize()
   &ScaleColor(
    Scale Alpha=&FloatFromCurve(FloatCurve=DataInterface() {propertyValues {Curve {keys [
         {interpMode=RCIM_Cubic, tangentMode=RCTM_User, time=0, value=0}
         {interpMode=RCIM_Cubic, tangentMode=RCTM_User, time=0.5, value=0.8}
         {tangentMode=RCTM_User, time=1, value=0}
        ]}}})
    ScaleA=true
    ScaleRGB=true
   )
   &CameraOffset(Camera Offset Amount=50)
  ]
  rendererClasses [&NiagaraSpriteRendererProperties]
  renderers [{rendererClass=&NiagaraSpriteRendererProperties}]
 }
 outwardsparks {
  bEnabled=false
  simTarget=GPUComputeSim
  emitterUpdateScript [
   &EmitterState()
   &SpawnRate(Spawn Probability=0.4, SpawnRate=15, Use Spawn Probability=true)
  ]
  particleSpawnScript [
   &InitializeParticle(
    Color=<rgba 2 0.2878 0.2281 0.65>
    Lifetime Max=0.3
    Lifetime Min=0.1
    Lifetime Mode=::NewEnumerator1
    Sprite Size Max=<1.5 3>
    Sprite Size Min=<0.75 1.25>
    Sprite Size Mode=::NewEnumerator2
    UsePositionOffset=false
    Write Material Random=true
   )
   &AddVelocityInCone(
    Cone Angle=5
    Cone Axis=<0.5 0 0>
    Module.Use Velocity Falloff On Cone Axis=true
    Velocity Distribution Along Cone Axis=0.75
    Velocity Falloff Away From Cone Axis=0.333
    Velocity Strength=&RandomRangeFloat(Minimum=250, Maximum=300)
   )
  ]
  particleUpdateScript [
   &ParticleState()
   &GravityForce(Gravity=<0 0 350>)
   &CurlNoiseForce(Noise Strength=300, Pan Noise=false)
   &Drag(Drag=&RandomRangeFloat(Minimum=-0.8, Maximum=-1.2), Use Linear Drag=true, Use Rotational Drag=false) {enabled=false}
   &ScaleColor(
    Scale Alpha=&FloatFromCurve(FloatCurve=DataInterface() {propertyValues {Curve {keys [
         {interpMode=RCIM_Cubic, tangentMode=RCTM_User, time=0, value=0}
         {interpMode=RCIM_Cubic, tangentMode=RCTM_User, time=0.5, value=1}
         {tangentMode=RCTM_User, time=1, value=0}
        ]}}})
    ScaleA=true
    ScaleRGB=true
   )
   &SolveForcesAndVelocity(Clamp Velocity=false, Limit Acceleration=false)
   &ScaleSpriteSizeBySpeed(
    Max Scale Factor=<1.5 5>
    Min Scale Factor=<0.5 2>
    Sample Scale Factor By Curve=true
    Scale Factor Curve=DataInterface() {propertyValues {Curve {keys [
        {arriveTangent=2.865, interpMode=RCIM_Cubic, time=0.1247, value=-0.0006}
        {arriveTangent=4.1522, interpMode=RCIM_Cubic, leaveTangent=4.1522, tangentMode=RCTM_User, time=0.3948, value=0.4893}
        <at 0.9958 0.9974 RCIM_Cubic>
       ]}}}
    Velocity Threshold=1500
   )
  ]
  rendererClasses [&NiagaraSpriteRendererProperties]
  renderers [{rendererClass=&NiagaraSpriteRendererProperties}]
 }
}
summary {
 systemName=NS_RepairToolCore
 emitters [
  {emitterName=CoreCone, rendererClasses [&NiagaraMeshRendererProperties]}
  {emitterName=CoreCone_Outer, rendererClasses [&NiagaraMeshRendererProperties]}
  {emitterName=FlickerCore, rendererClasses [&NiagaraMeshRendererProperties]}
  {emitterName=FlickerFlare, rendererClasses [&NiagaraSpriteRendererProperties, &NiagaraLightRendererProperties]}
  {emitterName=FlickerFlareWide, rendererClasses [&NiagaraSpriteRendererProperties]}
  {emitterName=CoreCone_Outer001, rendererClasses [&NiagaraMeshRendererProperties]}
  {bEnabled=false, emitterName=Smoke, rendererClasses [&NiagaraSpriteRendererProperties]}
  {bEnabled=false, emitterName=outwardsparks, simTarget=GPUComputeSim, rendererClasses [&NiagaraSpriteRendererProperties]}
 ]
 userVariables [
  User.EndTorchColorDamage {
   type=MaterialInterface
   description=""
   defaultValue {struct=NiagaraExt_VariableValue_Object, value {object=null, objectClass=/Script/Engine.MaterialInterface}}
  }
  User.EndTorchColorHeal {
   type=MaterialInterface
   description=""
   defaultValue {struct=NiagaraExt_VariableValue_Object, value {object=null, objectClass=&MaterialInterface}}
  }
  User.New Skeletal Mesh {
   type=NiagaraDataInterfaceSkeletalMesh
   defaultValue {
    struct=NiagaraExt_VariableValue_DataInterface
    value {dataInterface=&self:NiagaraDataInterfaceSkeletalMesh_0}
   }
  }
  User.TorchColor {type=LinearColor, description="", defaultValue {struct=LinearColor, value=<rgba 0 0.2285 2.5 0.75>}}
  User.TorchOn {type=NiagaraBool}
 ]
}
