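; Gameplay ability for the ResTorch repair tool that removes the torch from the player's inventory once it is depleted.
; When the ability ends, it calls the parent, then on authority checks whether the source weapon's remaining ammo is zero or less.
; If so, it stores the avatar as the player pawn and removes one item by the weapon's inventory GUID through that pawn's controller.
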(event Ability|EventOnEndAbility
  (params (bWasCancelled :type Boolean))
  (bind _returnvalue (Ability|GetAvatarActorfromActorInfo))
  ;[1 0.2969 0.2969] TORCH DEPLETE -- On Ability End (player stops firing the torch), we check if ammo <=0. If yes, we remove the Torch from the inventory
  (|CallParentFunction bWasCancelled)
  (Ability|HasAuthority
    (:True
      (if (<= (FortWeapon|GetRemainingAmmo (Ability|GetCurrentSourceWeapon)) 0)
        (Utilities|IsValid _returnvalue
          (:Is%20Valid
            (bind _asfort_player_pawn (Utilities|Casting|CastToFortPlayerPawn _returnvalue))
            (bind _output_get (Variables|Default|SetPlayerPawn _asfort_player_pawn))
            (Inventory|RemoveItemFromPlayerByGuid (Fortnite|GetFortPlayerControllerfromActor _output_get) (FortWeapon|GetInventoryGUID (Ability|GetCurrentSourceWeapon)) 1 true)))))))
